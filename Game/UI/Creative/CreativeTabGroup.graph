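# Widget Blueprint for the Creative UI tab bar, subclassing FortCreativeTabGroup.
# Lays out a horizontal, scrollbar-less Tab_Box between left and right input buttons bound to the gamepad shoulder actions.
# The side buttons call PreviousTab and NextTab on the parent class.
# When the parent adds a tab button, the event casts it to IconTextButton and sets its text, plus its icon for non-simple tabs.
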WidgetBlueprint /Game/UI/Creative/CreativeTabGroup
graphs [EventGraph {nodes [
   "On Button Clicked (UserActionMenuInputButton_Right)"!() -> (OutputDelegate, then, Button:Common Button Legacy Object Reference) {is=title K2Node_ComponentBoundEvent}
   > NextTab(self:Fort Creative Tab Group Object Reference)
   "On Button Clicked (UserActionMenuInputButton_Left)"!() -> (OutputDelegate, then, Button:Common Button Legacy Object Reference) {is=title K2Node_ComponentBoundEvent}
   > PreviousTab(self:Fort Creative Tab Group Object Reference)
   n1 = Event OnTabButtonAdded!() -> (OutputDelegate, then, Button:Common Button Base Object Reference, TabName:"Text (by ref)", TabIcon:"Slate Brush Structure (by ref)") {is=title K2Node_Event}
   > n2 = Cast To IconTextButton!(execute, Object=@n1.Button) -> (then, CastFailed, AsIcon Text Button:Icon Text Button Object Reference) {is=title K2Node_DynamicCast}
   > Set Text!(execute, self:Icon Text Button Object Reference=@n2, ButtonText:Text=@n1.TabName) -> (then) {is=title K2Node_CallFunction}
   > Set Icon!(execute, self:Icon Text Button Object Reference=@n2, IconBrush:Slate Brush Structure=@n1.TabIcon) -> (then) {is=title K2Node_CallFunction}
   n3 = Event OnSimpleTabButtonAdded!() -> (OutputDelegate, then, Button:Common Button Base Object Reference, TabName:"Text (by ref)") {is=title K2Node_Event}
   > n4 = Cast To IconTextButton(Object=@n3.Button)
   > Set Text(self=@n4, ButtonText=@n3.TabName)
  ]}]
variables [NewTab {type=IconTextButton_C, default=null}]
widgets {
 info {inheritedWidgetCount=1, parentClass=/Script/FortniteUI.FortCreativeTabGroup, rootWidgetClass=&Overlay, widgetCount=5}
 widgets [
  Overlay_0 {widgetClassPath=&Overlay, widget=Overlay_0}
  HorizontalBox_59 {widgetClassPath=&HorizontalBox, parent=Overlay_0, widget=HorizontalBox_59, slot {class=OverlaySlot}}
  UserActionMenuInputButton_Left {
   widgetClassPath=/Game/UI/SocialPanel/UserActionMenuInputButton
   bIsVariable=true
   parent=HorizontalBox_59
   widget=UserActionMenuInputButton_Left
   properties {
    onButtonBaseClicked=(&self:WidgetTree.UserActionMenuInputButton_Left.HandleOnButtonClicked)
    onButtonBaseDoubleClicked=(&self:WidgetTree.UserActionMenuInputButton_Left.HandleOnButtonDoubleClicked)
    onButtonBaseHovered=(&self:WidgetTree.UserActionMenuInputButton_Left.HandleOnButtonHovered)
    onButtonBaseUnhovered=(&self:WidgetTree.UserActionMenuInputButton_Left.HandleOnButtonUnhovered)
    onSelectedChangedBase=(&self:WidgetTree.UserActionMenuInputButton_Left.HandleOnSelectedChanged)
    triggeringInputAction {dataTable=&InputActionDataTable, rowName=GenericLeftShoulder}
   }
   slot {class=HorizontalBoxSlot, properties {padding=<ltrb 40 0 0 0>}}
  }
  Tab_Box {
   widgetClassPath=/Script/FortniteUI.AthenaScrollBox
   bInherited=true
   bIsVariable=true
   parent=HorizontalBox_59
   widget=Tab_Box
   properties {allowOverscroll=false, orientation=Orient_Horizontal, scrollBarVisibility=Collapsed}
   slot {class=HorizontalBoxSlot, properties {size {sizeRule=Fill, value=1}}}
  }
  UserActionMenuInputButton_Right {
   widgetClassPath=&UserActionMenuInputButton
   bIsVariable=true
   parent=HorizontalBox_59
   widget=UserActionMenuInputButton_Right
   properties {
    onButtonBaseClicked=(&self:WidgetTree.UserActionMenuInputButton_Right.HandleOnButtonClicked)
    onButtonBaseDoubleClicked=(&self:WidgetTree.UserActionMenuInputButton_Right.HandleOnButtonDoubleClicked)
    onButtonBaseHovered=(&self:WidgetTree.UserActionMenuInputButton_Right.HandleOnButtonHovered)
    onButtonBaseUnhovered=(&self:WidgetTree.UserActionMenuInputButton_Right.HandleOnButtonUnhovered)
    onSelectedChangedBase=(&self:WidgetTree.UserActionMenuInputButton_Right.HandleOnSelectedChanged)
    triggeringInputAction {dataTable=&InputActionDataTable, rowName=GenericRightShoulder}
   }
   slot {class=HorizontalBoxSlot, properties {horizontalAlignment=HAlign_Right, padding=<ltrb 0 0 40 0>}}
  }
 ]
}
