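# Base layout Widget Blueprint for the Creative island selection screen, built on the native FortCreativeIslandSelectScreen class.
# A blurred, safe-zoned panel stacks a gradient header, tab group and instant-transition widget switcher, with Cancel overlaid bottom-right and a touch-only close button.
# PreConstruct copies the Title string into the header; the graph holds no other logic.
# Tab group, switcher and Cancel fill inherited widget bindings, while Button_CloseMobile is left unassigned.
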WidgetBlueprint /FortniteCoreUI/UI/Creative/Creative_Island_Select_Base_UI
graphs [EventGraph {nodes [
   Event PreConstruct()
   > "SetText (Text)"(self=Get Title_Block() -> (Title_Block:Common Text Object Reference), InText=Get Title() -> (Title:Text))
   Event Construct()
   Event Tick()
  ]}]
variables [Title {type=string, default="{Please Set Title}"}]
widgets {
 info {
  inheritedWidgetCount=4
  parentClass=/Script/FortniteUI.FortCreativeIslandSelectScreen
  rootWidgetClass=&BackgroundBlur
  widgetCount=23
 }
 widgets [
  BackgroundBlur_34 {
   widgetClassPath=&BackgroundBlur
   widget=BackgroundBlur_34
   properties {
    blurStrength=25
    lowQualityFallbackBrush {
     bIsDynamicallyLoaded=false
     drawAs=Image
     imageType=NoImage
     mirroring=NoMirror
     resourceName=None
     resourceObject=null
     tiling=NoTile
     imageSize=<0 0>
     margin=<ltrb 0 0 0 0>
     outlineSettings {
      bUseBrushTransparency=false
      roundingType=HalfHeightRadius
      width=0
      color {colorUseRule=UseColor_Specified, specifiedColor=<rgba 0 0 0 0>}
      cornerRadii=<0 0 0 0>
     }
     tintColor {colorUseRule=UseColor_Specified, specifiedColor=<rgba 1 1 1 0.8>}
     uVRegion {bIsValid=false, max=<0 0>, min=<0 0>}
    }
   }
  }
  SafeZone_0 {
   widgetClassPath=&SafeZone
   bIsVariable=true
   parent=BackgroundBlur_34
   widget=SafeZone_0
   slot {class=BackgroundBlurSlot, properties {padding=<ltrb 0 0 0 0>}}
  }
  Overlay_3 {widgetClassPath=&Overlay, parent=SafeZone_0, widget=Overlay_3, properties {visibility=Visible}, slot {class=SafeZoneSlot}}
  Image_Shadow {
   widgetClassPath=&Image
   bIsVariable=true
   parent=Overlay_3
   widget=Image_Shadow
   properties {colorAndOpacity=<rgba 0.0052 0.0052 0.0056 0.4>, renderTransform {angle=0, scale=<1 1>, shear=<0 0>, translation=<12 12>}}
   slot {class=OverlaySlot, properties {horizontalAlignment=HAlign_Fill, verticalAlignment=VAlign_Fill}}
  }
  Overlay_7 {
   widgetClassPath=&Overlay
   parent=Overlay_3
   widget=Overlay_7
   slot {class=OverlaySlot, properties {horizontalAlignment=HAlign_Fill, verticalAlignment=VAlign_Fill}}
  }
  CommonBorder_0 {
   widgetClassPath=&CommonBorder
   parent=Overlay_7
   widget=CommonBorder_0
   properties {
    background {
     bIsDynamicallyLoaded=false
     drawAs=Box
     imageType=NoImage
     mirroring=NoMirror
     resourceName=None
     resourceObject=null
     tiling=NoTile
     imageSize=<32 32>
     margin=<ltrb 0 0 0 0>
     outlineSettings {
      bUseBrushTransparency=false
      roundingType=HalfHeightRadius
      width=0
      color {colorUseRule=UseColor_Specified, specifiedColor=<rgba 0 0 0 0>}
      cornerRadii=<0 0 0 0>
     }
     tintColor {colorUseRule=UseColor_Specified, specifiedColor=<rgba 0 0.0222 0.2874 1>}
     uVRegion {bIsValid=false, max=<0 0>, min=<0 0>}
    }
    padding=<ltrb 0 0 0 0>
   }
   slot {class=OverlaySlot, properties {horizontalAlignment=HAlign_Fill, verticalAlignment=VAlign_Fill}}
  }
  Overlay_6 {widgetClassPath=&Overlay, parent=CommonBorder_0, widget=Overlay_6, slot {class=BorderSlot, properties {padding=<ltrb 0 0 0 0>}}}
  VerticalBox_1 {
   widgetClassPath=&VerticalBox
   parent=Overlay_6
   widget=VerticalBox_1
   properties {navigation=&self:WidgetTree.VerticalBox_1.WidgetNavigation_0}
   slot {class=OverlaySlot, properties {horizontalAlignment=HAlign_Fill, verticalAlignment=VAlign_Fill}}
  }
  SizeBox_Header {
   widgetClassPath=&SizeBox
   parent=VerticalBox_1
   widget=SizeBox_Header
   properties {bOverride_HeightOverride=true, heightOverride=132}
   slot {class=VerticalBoxSlot, properties {padding=<ltrb 0 -2 0 0>}}
  }
  Overlay {widgetClassPath=&Overlay, parent=SizeBox_Header, widget=Overlay, slot {class=SizeBoxSlot}}
  Image {
   widgetClassPath=&Image
   parent=Overlay
   widget=Image
   properties {
    brush {
     bIsDynamicallyLoaded=false
     drawAs=Image
     imageType=NoImage
     mirroring=NoMirror
     resourceName=None
     resourceObject=/Game/Athena/UI/InGame/Creative/Art/M_UI_BoltGradientBorder_Creative_IslandSelectHeaderTop
     tiling=NoTile
     imageSize=<32 85.201>
     margin=<ltrb 0 0 0 0>
     outlineSettings {
      bUseBrushTransparency=false
      roundingType=HalfHeightRadius
      width=0
      color {colorUseRule=UseColor_Specified, specifiedColor=<rgba 0 0 0 0>}
      cornerRadii=<0 0 0 0>
     }
     tintColor {colorUseRule=UseColor_Specified, specifiedColor=<rgba 1 1 1 1>}
     uVRegion {bIsValid=false, max=<0 0>, min=<0 0>}
    }
    renderTransform {angle=0, scale=<1 -1>, shear=<0 0>, translation=<0 0>}
   }
   slot {class=OverlaySlot, properties {horizontalAlignment=HAlign_Fill, padding=<ltrb 0 -4 0 0>}}
  }
  Image_180 {
   widgetClassPath=&Image
   parent=Overlay
   widget=Image_180
   properties {
    brush {
     bIsDynamicallyLoaded=false
     drawAs=Image
     imageType=NoImage
     mirroring=NoMirror
     resourceName=None
     resourceObject=/Game/Athena/UI/InGame/Creative/Art/M_UI_BoltGradientBorder_Creative_IslandSelectHeader
     tiling=NoTile
     imageSize=<32 79.1332>
     margin=<ltrb 0 0 0 0>
     outlineSettings {
      bUseBrushTransparency=false
      roundingType=HalfHeightRadius
      width=0
      color {colorUseRule=UseColor_Specified, specifiedColor=<rgba 0 0 0 0>}
      cornerRadii=<0 0 0 0>
     }
     tintColor {colorUseRule=UseColor_Specified, specifiedColor=<rgba 1 1 1 1>}
     uVRegion {bIsValid=false, max=<0 0>, min=<0 0>}
    }
    renderTransform {angle=0, scale=<1 -1>, shear=<0 0>, translation=<0 0>}
   }
   slot {class=OverlaySlot, properties {horizontalAlignment=HAlign_Fill, verticalAlignment=VAlign_Bottom, padding=<ltrb 0 -3.2 0 0>}}
  }
  Title_Block {
   widgetClassPath=&CommonTextBlock
   bIsVariable=true
   parent=Overlay
   widget=Title_Block
   properties {
    justification=Center
    text="{Title}"
    colorAndOpacity {colorUseRule=UseColor_Specified, specifiedColor=<rgba 1 1 1 0.8>}
    font {
     bForceMonospaced=false
     bMaterialIsStencil=false
     fontMaterial=null
     fontObject=&Burbank
     letterSpacing=0
     monospacedWidth=1
     size=64
     skewAmount=0
     typefaceFontName=Burbank-Black
     outlineSettings {
      bApplyOutlineToDropShadows=false
      bMiteredCorners=false
      bSeparateFillAlpha=false
      outlineMaterial=null
      outlineSize=0
      outlineColor=<rgba 0 0 0 1>
     }
    }
    shadowOffset=<4.9266 5.2555>
   }
   slot {class=OverlaySlot, properties {horizontalAlignment=HAlign_Fill, verticalAlignment=VAlign_Center}}
  }
  Spacer_235 {widgetClassPath=&Spacer, parent=VerticalBox_1, widget=Spacer_235, properties {size=<1 24>}, slot {class=VerticalBoxSlot}}
  Overlay_0 {widgetClassPath=&Overlay, parent=VerticalBox_1, widget=Overlay_0, slot {class=VerticalBoxSlot}}
  Tab_Group {
   widgetClassPath=/Game/UI/Creative/CreativeTabGroup
   bInherited=true
   bIsVariable=true
   parent=Overlay_0
   widget=Tab_Group
   slot {class=OverlaySlot, properties {horizontalAlignment=HAlign_Center}}
  }
  Spacer_780 {widgetClassPath=&Spacer, parent=VerticalBox_1, widget=Spacer_780, properties {size=<1 12>}, slot {class=VerticalBoxSlot}}
  Overlay_8 {
   widgetClassPath=&Overlay
   bIsVariable=true
   parent=VerticalBox_1
   widget=Overlay_8
   slot {class=VerticalBoxSlot, properties {size {sizeRule=Fill, value=1}}}
  }
  Widget_Switcher {
   widgetClassPath=&CommonActivatableWidgetSwitcher
   bInherited=true
   bIsVariable=true
   parent=Overlay_8
   widget=Widget_Switcher
   properties {transitionDuration=0}
   slot {class=OverlaySlot, properties {horizontalAlignment=HAlign_Fill, verticalAlignment=VAlign_Fill}}
  }
  Button_Cancel {
   widgetClassPath=&IconTextButton
   bInherited=true
   bIsVariable=true
   parent=Overlay_8
   widget=Button_Cancel
   properties {
    bIsFocusable=false
    buttonText=CANCEL
    contentAlignment=HAlign_Center
    minWidth=172
    onButtonBaseClicked=(&self:WidgetTree.Button_Cancel.HandleOnButtonClicked)
    onButtonBaseDoubleClicked=(&self:WidgetTree.Button_Cancel.HandleOnButtonDoubleClicked)
    onButtonBaseHovered=(&self:WidgetTree.Button_Cancel.HandleOnButtonHovered)
    onButtonBaseUnhovered=(&self:WidgetTree.Button_Cancel.HandleOnButtonUnhovered)
    onSelectedChangedBase=(&self:WidgetTree.Button_Cancel.HandleOnSelectedChanged)
    style=/FortUILibrary/Foundation/Buttons/ButtonStyles/Athena/Athena_ButtonStyle_AngledBlueMenuButton.Athena_ButtonStyle_AngledBlueMenuButton_C
    triggeringInputAction {dataTable=/FortUILibrary/Input/CreativeInputActionDataTable, rowName=Cancel}
   }
   slot {class=OverlaySlot, properties {horizontalAlignment=HAlign_Right, verticalAlignment=VAlign_Bottom, padding=<ltrb 0 0 24 16>}}
  }
  VisibilityBorder_Touch {
   widgetClassPath=&CommonHardwareVisibilityBorder
   parent=Overlay_3
   widget=VisibilityBorder_Touch
   properties {visibilityQuery {
     autoDescription=" ANY( Input.Touch )"
     tokenStreamVersion=0
     userDescription=""
     queryTokenStream [0,1,1,1,0]
     tagDictionary [{tagName=Input.Touch}]
    }}
   slot {class=OverlaySlot, properties {horizontalAlignment=HAlign_Right, padding=<ltrb 0 16 16 0>}}
  }
  Button_CloseTouch {
   widgetClassPath=/FortUILibrary/Foundation/Buttons/CloseButton
   bIsVariable=true
   parent=VisibilityBorder_Touch
   widget=Button_CloseTouch
   properties {
    onButtonBaseClicked=(&self:WidgetTree.Button_CloseTouch.HandleOnButtonClicked)
    onButtonBaseDoubleClicked=(&self:WidgetTree.Button_CloseTouch.HandleOnButtonDoubleClicked)
    onButtonBaseHovered=(&self:WidgetTree.Button_CloseTouch.HandleOnButtonHovered)
    onButtonBaseUnhovered=(&self:WidgetTree.Button_CloseTouch.HandleOnButtonUnhovered)
    onSelectedChangedBase=(&self:WidgetTree.Button_CloseTouch.HandleOnSelectedChanged)
   }
   slot {class=BorderSlot, properties {padding=<ltrb 0 0 0 0>}}
  }
  Button_CloseMobile {widgetClassPath=&CommonButtonBase, bInherited=true, widget=null}
 ]
}
