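# Widget Blueprint for the Discovery island selection screen in Fortnite Creative.
# It derives from the native FortCreativeIslandSelectDiscovery class and uses the inherited Creative_Island_Select_Base_UI widget as its root, titled DISCOVER.
# It adds no logic or variables of its own, with only unconnected PreConstruct, Construct and Tick event stubs in the graph.
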WidgetBlueprint /FortniteCoreUI/UI/Creative/Creative_Island_Select_Discovery_UI
graphs [EventGraph {nodes [
   Event PreConstruct()
   Event Construct()
   Event Tick()
  ]}]
variables []
widgets {
 info {
  inheritedWidgetCount=1
  parentClass=/Script/FortniteUI.FortCreativeIslandSelectDiscovery
  rootWidgetClass=&/Creative_Island_Select_Base_UI.Creative_Island_Select_Base_UI_C
  widgetCount=1
 }
 widgets [IslandSelect_Base {
   widgetClassPath=&/Creative_Island_Select_Base_UI
   bInherited=true
   bIsVariable=true
   widget=IslandSelect_Base
   properties {title=DISCOVER}
  }]
}
