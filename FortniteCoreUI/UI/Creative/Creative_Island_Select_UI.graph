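# Creative "Choose Destination" screen for picking an island for a portal by code, recent list or favorites.
# A tab button group switches views and list filters, feeding the selected link into a details panel.
# Code validation drives a colored status banner and shake animation, while link data manager queries supply tags and thumbnails asynchronously.
# The initializing portal receives the link through an interface message, and cursor mode stays locked until deactivation.
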WidgetBlueprint /FortniteCoreUI/UI/Creative/Creative_Island_Select_UI
animations [
 Loading {
  binding_count=1
  length_seconds=1
  object=&self:Loading
  sequence {bindings [MatchmakingSpinner {
     id=#1
     possessed_class=Image
     tracks [{
       type=MovieSceneFloatTrack
       display_name=Render Opacity
       sections [{type=MovieSceneFloatSection, end_frame=20, start_frame=0, channels [None {type=MovieSceneScriptingFloatChannel, keys [[1,1],[20,0]]}]}]
      }]
    }]}
 }
 InvalidCode {
  binding_count=1
  length_seconds=0.5
  object=&self:InvalidCode
  sequence {bindings [Overlay_Code {
     id=#2
     possessed_class=Overlay
     tracks [{
       type=MovieScene2DTransformTrack
       sections [{
         type=MovieScene2DTransformSection
         end_frame=10
         start_frame=0
         channels [
          Translation.X {type=MovieSceneScriptingFloatChannel, keys [[0,0],[3,-10],[5,10],[7,-10],[9,10]]}
          Translation.Y {type=MovieSceneScriptingFloatChannel}
         ]
        }]
      }]
    }]}
 }
]
graphs [
 ClearDetailsPanel {nodes [
   ClearDetailsPanel() {is=title K2Node_FunctionEntry}
   > "SetText (Text)"(
    self=[
     Get Text_IslandDescription() -> (Text_IslandDescription:Common Text Object Reference)
     Get Text_Creator() -> (Text_Creator:Common Text Object Reference)
     Get Text_IslandName() -> (Text_IslandName:Common Text Object Reference)
     Get PublishedIslandCode() -> (PublishedIslandCode:Common Text Object Reference)
    ]
    InText=GetEmptyText!() -> (ReturnValue:Text)
   )
   > SetVisibility(
    self=[
     Get Text_CreatorLabel() -> (Text_CreatorLabel:Common Text Object Reference)
     Get MatchMaking_Line_Description() -> (MatchMaking_Line_Description:Common Text Object Reference)
     Get Border_IslandImage() -> (Border_IslandImage:Border Object Reference)
     Get MatchMaking_MaximumNumberOfPlayers() -> (MatchMaking_MaximumNumberOfPlayers:Common Text Object Reference)
     Get LightningViolator() -> (LightningViolator:Lightning Violator Object Reference)
     Get TagDisplay() -> (TagDisplay:Cr Tag Display Object Reference)
     Get ImageLoadingSpinner() -> (ImageLoadingSpinner:Image Object Reference)
    ]
    InVisibility=Collapsed
   )
  ]}
 EventGraph {nodes [
   "On Button Clicked (UserActionMenuInputButton_Left)"!() -> (OutputDelegate, then, Button:Common Button Legacy Object Reference) {is=title K2Node_ComponentBoundEvent}
   > MoveTabButtonSelection!(execute, self, Forward:bool=false) -> (then)
   n1 = "On Button Clicked (Button_CloseTouch)"!() -> (OutputDelegate, then, Button:Common Button Legacy Object Reference) {is=title K2Node_ComponentBoundEvent}
   ClearSelfFromPortal(execute=[@n1.then, @n16.then])
   "On Button Clicked (Button_Refresh)"!() -> (OutputDelegate, then, Button:Common Button Legacy Object Reference) {is=title K2Node_ComponentBoundEvent}
   > PlayAnimation(InAnimation=Get Loading() -> (Loading:Widget Animation Object Reference))
   n2 = "On Item Selection Changed (ListView_IslandLinks)"!() -> (OutputDelegate, then, Item:Object Reference, bIsSelected:bool) {is=title K2Node_ComponentBoundEvent}
   n3 = GetSelectedButtonIndex!(self:Common Button Group Base Object Reference=Get TabButtonGroup() -> (TabButtonGroup:Common Button Group Legacy Object Reference)) -> (ReturnValue:int=0)
   > Branch(Condition=OR Boolean(A="Equal (Integer)"(A=@n3, B=1) {is=title K2Node_CallFunction}, B="Equal (Integer)"(A=@n3, B=2)))
   > n4 = Cast To FortCreativeIslandLink!(execute, Object=@n2.Item) -> (then, CastFailed, AsFort Creative Island Link:Fort Creative Island Link Object Reference) {is=title K2Node_DynamicCast}
   > SetSelectedIslandLink!(execute, self, NewSelectedIslandLink:Fort Creative Island Link Object Reference=@n4) -> (then)
   > UpdateDetailsPanel!(
    execute
    self
    inIslandLink:Fort Creative Island Link Object Reference=Get SelectedIslandLink() -> (SelectedIslandLink:Fort Creative Island Link Object Reference)
    Empty:bool=false
   ) -> (then)
   "On Button Clicked (Tab_Favorites)"!() -> (OutputDelegate, then, Button:Common Button Legacy Object Reference) {is=title K2Node_ComponentBoundEvent}
   n5 = Get TabButtonGroup()
   > SelectButtonAtIndex!(execute, self:Common Button Group Base Object Reference=@n5, ButtonIndex:int=2, bAllowSound:bool=true) -> (then)
   "On Button Clicked (Button_Favorite)"!() -> (OutputDelegate, then, Button:Common Button Legacy Object Reference) {is=title K2Node_ComponentBoundEvent}
   > Is Valid(InputObject=Get SelectedIslandLink())
   n6 = Get SelectedIslandLink()
   > n7 = Branch(Condition="Equal (Enum)"(
     A:EFortCreativeIslandLinkCategory Enum=Get Category!(self:Fort Creative Island Link Object Reference=@n6) -> (Category:EFortCreativeIslandLinkCategory Enum=Default)
     B:EFortCreativeIslandLinkCategory Enum=Favorite
    ))
   n8 = Get SelectedIslandLink()
   > n9 = RemoveIslandFromFavorites!(execute, self:Fort Creative Island Link Screen Object Reference, LinkEntry:Fort Creative Island Link Object Reference=@n8) -> (then)
   RegenerateAllEntries(
    execute=[@n9.then, @n10.then]
    self:List View Base Object Reference=Get ListView_IslandLinks() -> (ListView_IslandLinks:List View Object Reference)
   )
   n10 = AddIslandToFavorites!(execute=@n7.else, self:Fort Creative Island Link Screen Object Reference, LinkEntry:Fort Creative Island Link Object Reference=@n8) -> (then)
   "On Button Clicked (UserActionMenuInputButton_Right)"!() -> (OutputDelegate, then, Button:Common Button Legacy Object Reference) {is=title K2Node_ComponentBoundEvent}
   > MoveTabButtonSelection(Forward=true)
   "On Button Clicked (Tab_EnterCode)"!() -> (OutputDelegate, then, Button:Common Button Legacy Object Reference) {is=title K2Node_ComponentBoundEvent}
   > SelectButtonAtIndex(self=@n5, ButtonIndex=0, bAllowSound=true)
   > CenterWidget!(execute, ToCenter:Widget Object Reference=Get EditableText_IslandLink() -> (EditableText_IslandLink:Editable Text Object Reference)) -> (then)
   > UpdateDetailsPanel(inIslandLink=Get LastLinkQueried() -> (LastLinkQueried:Fort Creative Island Link Object Reference), Empty=false)
   > SetIsEnabled(
    self=Get Button_Select() -> (Button_Select:Common Button Legacy Object Reference)
    bInIsEnabled="Equal (Enum)"(
     A:EFortActivityValidationResult Enum=Get LastValidateResult() -> (LastValidateResult:EFortActivityValidationResult Enum=NotFound)
     B:EFortActivityValidationResult Enum=Success
    )
   )
   "On Button Clicked (Tab_Recent)"!() -> (OutputDelegate, then, Button:Common Button Legacy Object Reference) {is=title K2Node_ComponentBoundEvent}
   > SelectButtonAtIndex(self=@n5, ButtonIndex=1, bAllowSound=true)
   n11 = Get ListView_IslandLinks()
   > n12 = Branch(Condition="integer <= integer"!(A:int=GetNumItems!(self:List View Object Reference=@n11) -> (ReturnValue:int=0), B:int=0) -> (ReturnValue:bool) {is=title K2Node_CallFunction})
   > UpdateDetailsPanel(inIslandLink=Get SelectedIslandLink(), Empty=true)
   n13 = Get ListView_IslandLinks()
   n14 = Branch(
    execute=@n12.else
    Condition="integer <= integer"(A=Get Num Items Selected!(self:List View Object Reference=@n13) -> (ReturnValue:int=0) {is=title K2Node_CallFunction}, B=0)
   )
   > SetSelectedIndex!(execute, self:List View Object Reference=Get ListView_IslandLinks(), Index:int=0) -> (then)
   n15 = Cast To FortCreativeIslandLink(
    execute=@n14.else
    Object=Get Selected Item!(self:List View Object Reference=Get ListView_IslandLinks()) -> (ReturnValue:Object Reference) {is=title K2Node_CallFunction}
   )
   > SetSelectedIslandLink(NewSelectedIslandLink=@n15)
   > UpdateDetailsPanel(inIslandLink=Get SelectedIslandLink(), Empty=false)
   n16 = "On Button Clicked (Button_Cancel)"!() -> (OutputDelegate, then, Button:Common Button Legacy Object Reference) {is=title K2Node_ComponentBoundEvent}
   n17 = CleanupAfterClose!() -> (OutputDelegate, then, IslandLink:Fort Creative Island Link Object Reference) {is=title K2Node_CustomEvent}
   > ClearSelfFromPortal()
   n18 = HandleButtonGroupSelectionChanged!() -> (OutputDelegate, then, AssociatedButton:Common Button Legacy Object Reference, ButtonIndex:int) {is=title K2Node_CustomEvent}
   > SetActiveWidgetIndex(
    self=Get Switcher_Tabs() -> (Switcher_Tabs:Common Widget Switcher Legacy Object Reference)
    Index=Select!(Option 0:int=0, Option 1:int=1, Option 2:int=1, Index:int=@n18.ButtonIndex) -> (ReturnValue:int)
   )
   > SetVisibility(
    self=Get HBox_ListActions() -> (HBox_ListActions:Horizontal Box Object Reference)
    InVisibility=Select(
     Option 0:ESlateVisibility Enum=Collapsed
     Option 1:ESlateVisibility Enum=SelfHitTestInvisible
     Option 2:ESlateVisibility Enum=SelfHitTestInvisible
     Index=@n18.ButtonIndex
    ) -> (ReturnValue:ESlateVisibility Enum)
   )
   > SetVisibility(
    self=Get Overlay_Remove() -> (Overlay_Remove:Overlay Object Reference)
    InVisibility=Select(Option 0=Collapsed, Option 1=SelfHitTestInvisible, Option 2=Collapsed, Index=@n18.ButtonIndex)
   )
   > SetFavoritesOnlyMode!(execute, self:Fort Creative Island Link Screen Object Reference, bFavoritesOnly:bool=true) -> (then)
   > n19 = Switch on Int!(execute:Exec, Selection:int=@n18.ButtonIndex) -> (Default:Exec, 0:Exec, 1:Exec, 2:Exec)
   CenterWidget(execute=@n19.0, ToCenter=Get EditableText_IslandLink())
   > UpdateDetailsPanel(inIslandLink=Get LastLinkQueried(), Empty=false)
   > SetBrush(self=Get Image_Island() -> (Image_Island:Image Object Reference), InBrush=Get DefaultArt() -> (DefaultArt:Slate Brush Structure))
   > SetVisibility(self=Get ImageLoadingSpinner(), InVisibility=Collapsed)
   > SetVisibility(self=Get Border_IslandImage(), InVisibility=SelfHitTestInvisible)
   > UpdateUIfromLinkCode!(
    execute
    self
    LinkCode:Online Link Id Structure=GetLinkIdFromInput!(self:Fort Creative Island Link Screen Object Reference) -> (ReturnValue:Online Link Id Structure)
   ) -> (then)
   n20 = SetFavoritesOnlyMode(execute=@n19.1, bFavoritesOnly=false)
   SetFocus(execute=[@n20.then, @n26.then], self=Get ListView_IslandLinks())
   n21 = Get ListView_IslandLinks()
   > n22 = Branch(Condition="integer <= integer"(A=GetNumItems(self=@n21) -> (ReturnValue=0), B=0))
   > UpdateDetailsPanel(Empty=true)
   n23 = Get ListView_IslandLinks()
   n24 = Branch(execute=@n22.else, Condition="integer <= integer"(A=Get Num Items Selected(self=@n23) -> (ReturnValue=0), B=0))
   > SetSelectedIndex(self=Get ListView_IslandLinks(), Index=0)
   n25 = Cast To FortCreativeIslandLink(execute=@n24.else, Object=Get Selected Item(self=Get ListView_IslandLinks()))
   > SetSelectedIslandLink(NewSelectedIslandLink=@n25)
   > UpdateDetailsPanel(inIslandLink=Get SelectedIslandLink(), Empty=false)
   n26 = SetFavoritesOnlyMode(execute=@n19.2, bFavoritesOnly=true)
   n27 = HandleLinkCodeQuery!() -> (OutputDelegate, then, bSuccess:bool=~, ErrorMessage:String, LinkData:Fort Game Activity Object Reference) {is=title K2Node_CustomEvent}
   > n28 = Branch(Condition=@n27.bSuccess)
   > UpdateIslandArtFromLink!(execute, self, LinkData:Fort Game Activity Object Reference=@n27.LinkData) -> (then)
   n29 = Get TagDisplay()
   > SetVisibility(self=@n29, InVisibility=SelfHitTestInvisible)
   n30 = GetCreativeDescriptionTags!(self:Fort Game Activity Object Reference=@n27.LinkData) -> (ReturnValue:Array of Creative Island Description Tag Structures)
   > PopulateTags!(execute, self:Cr Tag Display Object Reference=@n29, inTags:Array of Creative Island Description Tag Structures=@n30) -> (then)
   > PrintString!(
    execute
    InString="To String (Integer)"(InInt=Length(TargetArray:Array of Creative Island Description Tag Structures=@n30))
    bPrintToScreen
    bPrintToLog
    TextColor
    Duration
    Key
   ) -> (then)
   PrintString(execute=@n28.else, InString=@n27.ErrorMessage)
   n31 = UpdateUIfromLinkCode!() -> (OutputDelegate, then, LinkCode:Online Link Id Structure) {is=title K2Node_CustomEvent}
   n32 = Break Online Link Id!(OnlineLinkId:"Online Link Id Structure (by ref)"=@n31.LinkCode) -> (Mnemonic:String, Version:int) {is=title K2Node_BreakStruct}
   > n33 = Branch(Condition="Not Equal Exactly (String)"(B=@n32))
   > QueryLinkDataByCode!(
    execute
    self:Fort Link Data Manager Object Reference=Get FortLinkDataManager() -> (ReturnValue:Fort Link Data Manager Object Reference) {is=title K2Node_GetSubsystem}
    LinkId:"Online Link Id Structure (by ref)"=@n31.LinkCode
    Callback:"Delegate (by ref)"=@n27
    bCalledByModerator:bool=false
   ) -> (then)
   > PrintString(InString=Append(A=@n32, B="Build String (Integer)"!(AppendTo:String="?v=", Prefix:String, InInt:int=@n32.Version, Suffix:String) -> (ReturnValue:String)))
   ClearDetailsPanel(execute=@n33.else) {is=title K2Node_CallFunction}
   n34 = UpdateIslandArtFromLink!() -> (OutputDelegate, then, LinkData:Fort Game Activity Object Reference) {is=title K2Node_CustomEvent}
   n35 = HandleUpdateIslandArt!() -> (OutputDelegate, then, bSuccess:bool=~, InThumbnailTexture:Texture 2DDynamic Object Reference) {is=title K2Node_CustomEvent}
   DownloadThumbnail!(
    execute=@n34.then
    self:Fort Link Data Manager Object Reference=Get FortLinkDataManager()
    LinkData:Fort Game Activity Object Reference=@n34.LinkData
    Callback:"Delegate (by ref)"=@n35
   ) -> (then)
   n36 = UpdateDetailsPanel!() -> (OutputDelegate, then, inIslandLink:Fort Creative Island Link Object Reference, Empty:bool=~) {is=title K2Node_CustomEvent}
   > SetSelectedIslandLink(NewSelectedIslandLink=@n36.inIslandLink)
   n37 = IsValid(Object=Get SelectedIslandLink())
   > n38 = Branch(Condition=OR Boolean(A=@n36.Empty, B=NOT Boolean(A=@n37)))
   > ClearDetailsPanel()
   n39 = Get SelectedIslandLink()
   "SetText (Text)"(
    execute=@n38.else
    self=Get Text_IslandDescription()
    InText="To Text (String)"(InString=Get Tagline!(self:Fort Creative Island Link Object Reference=@n39) -> (Tagline:String))
   )
   > "SetText (Text)"(
    self=Get Text_Creator()
    InText="To Text (String)"(InString=Get CreatorUsername!(self:Fort Creative Island Link Object Reference=@n39) -> (CreatorUsername:String))
   )
   > "SetText (Text)"(
    self=Get Text_IslandName()
    InText="To Text (String)"(InString=Get Title!(self:Fort Creative Island Link Object Reference=@n39) -> (Title:String))
   )
   n40 = Break Online Link Id(OnlineLinkId=Get LinkId!(self:Fort Creative Island Link Object Reference=@n39) -> (LinkId:Online Link Id Structure))
   n41 = Len(S=@n40)
   > SetVisibility(
    self=Get PublishedIslandCode()
    InVisibility=Select!(Option 0=Collapsed, Option 1=SelfHitTestInvisible, Index:bool="integer > integer"(A=@n41, B=0) {is=title K2Node_CallFunction}) -> (ReturnValue)
   )
   > "SetText (Text)"(self=Get PublishedIslandCode(), InText="To Text (String)"(InString=@n40))
   > SetVisibility(self=Get Text_CreatorLabel(), InVisibility=SelfHitTestInvisible)
   n42 = Get ImageUrl!(self:Fort Creative Island Link Object Reference=@n39) -> (ImageUrl:String)
   > n43 = Branch(Condition="Equal (Integer)"(A=Len(S=@n42), B=0))
   > SetBrush(self=Get Image_Island(), InBrush=Get DefaultArt())
   > SetVisibility(self=Get ImageLoadingSpinner(), InVisibility=Collapsed)
   > SetVisibility(self=Get Border_IslandImage(), InVisibility=SelfHitTestInvisible)
   n44 = DownloadImage!(execute:Exec=@n43.else, URL:String=@n42) -> (then:Exec, OnSuccess:Exec, OnFail:Exec, Texture:Texture 2DDynamic Object Reference) {is=title K2Node_AsyncAction}
   > SetVisibility(self=Get Border_IslandImage(), InVisibility=Hidden)
   > SetVisibility(self=Get ImageLoadingSpinner(), InVisibility=SelfHitTestInvisible)
   SetBrush(
    execute=@n44.OnSuccess
    self=Get Image_Island()
    InBrush=Make Slate Brush!(
     TintColor:Slate Color Structure=(SpecifiedColor=(R=1,G=1,B=1,A=1),ColorUseRule=UseColor_Specified)
     DrawAs:ESlateBrushDrawType Enum=Image
     Tiling:ESlateBrushTileType Enum=NoTile
     Mirroring:ESlateBrushMirrorType Enum=NoMirror
     ImageSize:Vector 2D Structure=(X=1024,Y=512)
     Margin:Margin Structure=(Left=0,Top=0,Right=0,Bottom=0)
     ResourceObject:Object Reference=@n44
     OutlineSettings:Slate Brush Outline Settings Structure=(CornerRadii=(X=0,Y=0,Z=0,W=0),Color=(SpecifiedColor=(R=0,G=0,B=0,A=0),ColorUseRule=UseColor_Specified),Width=0,RoundingType=HalfHeightRadius,bUseBrushTransparency=False)
    ) -> (SlateBrush:Slate Brush Structure) {is=title K2Node_MakeStruct}
   )
   > SetVisibility(self=Get ImageLoadingSpinner(), InVisibility=Collapsed)
   > SetVisibility(self=Get Border_IslandImage(), InVisibility=SelfHitTestInvisible)
   SetBrush(execute=@n44.OnFail, self=Get Image_Island(), InBrush=Get DefaultArt())
   > SetVisibility(self=Get ImageLoadingSpinner(), InVisibility=Collapsed)
   > SetVisibility(self=Get Border_IslandImage(), InVisibility=SelfHitTestInvisible)
   n45 = OnLinkModified() {is=title K2Node_CustomEvent}
   > SetVisibility(self=Get Overlay_Violator() -> (Overlay_Violator:Overlay Object Reference), InVisibility=Collapsed)
   > ClearDetailsPanel()
   > SetSelectedIslandLink()
   n46 = Close() {is=title K2Node_CustomEvent}
   > PopContentWidget!(execute) -> (then)
   ClearSelfFromPortal() {is=title K2Node_CustomEvent}
   > n47 = Get!(execute:Exec) -> (then:Exec, else:Exec, Portal:BGA Island Portal Object Reference)
   > Is Valid(InputObject=@n47)
   > ClearDisplayedWidget(self:BGA Island Portal Object Reference=@n47)
   n48 = Cast To FortPlayerController(Object=GetOwningPlayer())
   > SendIslandLink!(
    execute:Exec
    self:Object Reference=Get Init Object() -> (Init Object:Object Reference)
    Requesting Player Controller:Fort Player Controller Object Reference=@n48
    Link:Fort Creative Island Link Object Reference=Get SelectedIslandLink()
   ) -> (then:Exec, " ":bool) {is=title K2Node_Message}
   Event Construct()
   > Bind Event to On Editable Text Island Link Modified(self:Fort Creative Island Link Screen Object Reference, Delegate=@n45) {is=title K2Node_AddDelegate}
   > n49 = Construct Object from Class!(execute, Class=/Script/CommonUILegacy.CommonButtonGroupLegacy, self=Self-Reference(), bSelectionRequired:bool=true) -> (then, ReturnValue:Common Button Group Legacy Object Reference)
   > Set TabButtonGroup(TabButtonGroup:Common Button Group Legacy Object Reference=@n49) -> (Output_Get:Common Button Group Legacy Object Reference)
   > SetSelectionRequired!(execute, self:Common Button Group Base Object Reference=Get TabButtonGroup(), bRequireSelection:bool=false) -> (then)
   > DeselectAll(self:Common Button Group Base Object Reference=Get TabButtonGroup())
   > RemoveAll(self:Common Widget Group Base Object Reference=Get TabButtonGroup())
   n50 = Get TabButtonGroup()
   > AddWidget!(
    execute
    self:Common Widget Group Base Object Reference=@n50
    InWidget:Widget Object Reference=Get Tab_EnterCode() -> (Tab_EnterCode:Icon Text Button Object Reference)
   ) -> (then)
   > AddWidget(self=@n50, InWidget=Get Tab_Recent() -> (Tab_Recent:Icon Text Button Object Reference))
   > AddWidget(self=@n50, InWidget=Get Tab_Favorites() -> (Tab_Favorites:Icon Text Button Object Reference))
   > Bind Event to On Selected Button Changed(self:Common Button Group Legacy Object Reference=Get TabButtonGroup(), Delegate=@n18) {is=title K2Node_AddDelegate}
   > SetSelectionRequired(self=Get TabButtonGroup(), bRequireSelection=true)
   > InitUI()
   Event OnCreativeIslandLinksPopulated() {is=title K2Node_Event}
   n51 = GetNumItems(self=Get ListView_IslandLinks()) -> (ReturnValue=0)
   n52 = GetSelectedButtonIndex(self=Get TabButtonGroup()) -> (ReturnValue=0)
   n53 = Get NoFavorites() -> (NoFavorites:Overlay Object Reference)
   n54 = Get NoIslandsContent() -> (NoIslandsContent:Overlay Object Reference)
   > SetActiveWidget(
    self:Widget Switcher Object Reference=Get Switcher_IslandsList() -> (Switcher_IslandsList:Common Activatable Widget Switcher Object Reference)
    Widget=Select(
     Option 0:Widget Object Reference=Select!(Option 0=@n54, Option 1=@n54, Option 2:Widget Object Reference=@n53, Index:int=@n52) -> (ReturnValue)
     Option 1:Widget Object Reference=Get IslandListContent() -> (IslandListContent:Overlay Object Reference)
     Index="integer > integer"(A=@n51, B=0)
    ) -> (ReturnValue:Widget Object Reference)
   )
   n55 = Event InitFromObject!() -> (OutputDelegate, then, InitObject:Object Reference) {is=title K2Node_Event}
   > Set Initialization Object!(execute, self, Object:Object Reference=@n55.InitObject) -> (then) {is=title K2Node_CallFunction}
   n56 = Get FortHUDContext!() -> (ReturnValue:Fort HUDContext Object Reference) {is=title K2Node_GetBlueprintContextEx}
   > EnterCursorMode!(execute, self:Fort HUDContext Object Reference=@n56, ActionName:Name=None, CursorModeWidget:User Widget Object Reference=Self-Reference()) -> (then)
   > SetCursorModeLocked!(execute, self:Fort HUDContext Object Reference=@n56, bLocked:bool=true) -> (then)
   > Bind Event to On Creative Island Code Confirmed(self:Fort Creative Island Link Screen Object Reference, Delegate=@n17) {is=title K2Node_AddDelegate}
   > n57 = Get()
   > Bind Event to Close Shown UI(self:BGA Island Portal Object Reference=@n57, Delegate=@n46) {is=title K2Node_AddDelegate}
   n58 = Event OnCreativeIslandLinkValidated!() -> (OutputDelegate, then, Error:EFortActivityValidationResult Enum=NotFound, IsValidModerationState:bool) {is=title K2Node_Event}
   > Set LastValidateResult(LastValidateResult:EFortActivityValidationResult Enum=@n58.Error) -> (Output_Get:EFortActivityValidationResult Enum=NotFound)
   n59 = Get EditableText_IslandLink()
   > n60 = Branch(Condition="Equal Exactly (Text)"!(
     A:"Text (by ref)"="GetText (Editable Text)"!(self:Editable Text Object Reference=@n59) -> (ReturnValue:Text)
     B:"Text (by ref)"=GetEmptyText()
    ) -> (ReturnValue:bool))
   n61 = Sequence(execute=@n60.else)
   n62 = Get LightningViolator()
   > SetVisibility(self=@n62, InVisibility=SelfHitTestInvisible)
   > EventSetText!(
    execute
    self:Lightning Violator Object Reference=@n62
    NewText:Text=Select!(
     Option 0:Text="LOCTEXT(\"Try Again - Island Not Found\")"
     Option 1:Text="LOCTEXT(\"Try Again - Too Short\")"
     Option 2:Text="LOCTEXT(\"Try Again - Invalid Characters\")"
     Option 3:Text="LOCTEXT(\"Party is too Large\")"
     Option 4:Text="LOCTEXT(\"Island Private\")"
     Option 5:Text="LOCTEXT(\"ISLAND FOUND!\")"
     Option 6:Text
     Option 7:Text
     Index:Byte=@n58.Error
    ) -> (ReturnValue:Text)
   ) -> (then)
   n63 = Get ViolatorColor_Error() -> (ViolatorColor_Error:Linear Color Structure)
   > EventColorBG!(
    execute
    self:Lightning Violator Object Reference=@n62
    NewColor:Linear Color Structure=Select!(
     NotFound:Linear Color Structure=@n63
     InvalidKeyTooShort:Linear Color Structure=@n63
     InvalidKeyCharacters:Linear Color Structure=@n63
     IneligibleParty:Linear Color Structure=@n63
     IslandPrivate:Linear Color Structure=@n63
     Success:Linear Color Structure=Get ViolatorColor_Success() -> (ViolatorColor_Success:Linear Color Structure)
     Index:EFortActivityValidationResult Enum=@n58.Error
    ) -> (ReturnValue:Linear Color Structure)
   ) -> (then)
   > SetVisibility(self=Get Overlay_Violator(), InVisibility=SelfHitTestInvisible)
   n64 = Branch(
    execute=@n61.then_1
    Condition="Not Equal (Enum)"!(A:EFortActivityValidationResult Enum=@n58.Error, B:EFortActivityValidationResult Enum=Success) -> (ReturnValue:bool)
   )
   > PlayAnimation(InAnimation=Get InvalidCode() -> (InvalidCode:Widget Animation Object Reference), PlaybackSpeed=1.5)
   > n65 = Branch(Condition="Equal (Enum)"(A=@n58.Error, B=IneligibleParty))
   n66 = Get LastLinkQueried()
   UpdateDetailsPanel(execute=[@n65.then, @n64.else], inIslandLink=@n66, Empty=false) {is=title K2Node_CallFunction}
   > UpdateUIfromLinkCode(LinkCode=Get LinkId(self=@n66)) {is=title K2Node_CallFunction}
   Event On Activated()
   > HandleButtonGroupSelectionChanged!(execute, self, AssociatedButton:Common Button Legacy Object Reference, ButtonIndex:int=0) -> (then) {is=title K2Node_CallFunction}
   > CenterWidget(ToCenter=Get EditableText_IslandLink())
   Event PreConstruct()
   n67 = GetRuntimeOptions()
   > SetHintText!(
    execute
    self:Editable Text Object Reference=Get EditableText_IslandLink()
    InHintText:Text="To Text (String)"(InString=Get IslandCodeLinkMnemonicExampleText!(self:Fort Runtime Options Object Reference=@n67) -> (IslandCodeLinkMnemonicExampleText:String))
   ) -> (then)
   > "SetText (Text)"(
    self=Get Text_WebURL() -> (Text_WebURL:Common Text Object Reference)
    InText="To Text (String)"(InString=Get IslandCodeLinkURLText!(self:Fort Runtime Options Object Reference=@n67) -> (IslandCodeLinkURLText:String))
   )
   Event On Deactivated()
   n68 = Get FortHUDContext()
   > SetCursorModeLocked(self=@n68, bLocked=false)
   > EnterCameraMode(self:Fort HUDContext Object Reference=@n68)
   SetVisibility(execute=@n70.then, self=Get ImageLoadingSpinner(), InVisibility=Collapsed)
   SetRenderOpacity(execute=@n69.then, self=Get Image_Island(), InOpacity=1)
   "On Button Clicked (Button_Select)"!() -> (OutputDelegate, then, Button:Common Button Legacy Object Reference) {is=title K2Node_ComponentBoundEvent}
   SetBrush(
    execute=@n71.then
    self=Get Image_Island()
    InBrush=Make Slate Brush(
     TintColor=(SpecifiedColor=(R=1,G=1,B=1,A=1),ColorUseRule=UseColor_Specified)
     DrawAs=Image
     Tiling=NoTile
     Mirroring=NoMirror
     ImageSize=(X=1024,Y=512)
     Margin=(Left=0,Top=0,Right=0,Bottom=0)
     ResourceObject=@n35.InThumbnailTexture
     OutlineSettings=(CornerRadii=(X=0,Y=0,Z=0,W=0),Color=(SpecifiedColor=(R=0,G=0,B=0,A=0),ColorUseRule=UseColor_Specified),Width=0,RoundingType=HalfHeightRadius,bUseBrushTransparency=False)
    )
   )
   > n69 = SetVisibility(self=Get ImageLoadingSpinner(), InVisibility=Collapsed)
   n70 = SetBrush(execute=@n71.else, self=Get Image_Island(), InBrush=Get DefaultArt())
   n71 = Branch(execute=@n35.then, Condition=@n35.bSuccess)
  ]}
 GetDataFromEntry {nodes [
   n1 = GetDataFromEntry!() -> (then, Object:Object Reference) {is=title K2Node_FunctionEntry}
   > n2 = Cast To FortCreativeIslandLink(Object=@n1)
   n3 = Get LinkId(self=@n2)
   > Return Node!(
    execute
    CreatorName:Text="To Text (String)"(InString=Get CreatorUsername(self=@n2))
    Description:Text="To Text (String)"(InString=Get Tagline(self=@n2))
    IslandName:Text="To Text (String)"(InString=Get Title(self=@n2))
    Version:Text=Format Text!(Format="LOCTEXT(\"v{version}\")", version:int=Break Online Link Id(OnlineLinkId=@n3).Version) -> (Result)
   )
  ]}
 InitUI {nodes [
   InitUI() {is=title K2Node_FunctionEntry}
   > "SetText (Text)"(self=Get Text_IslandDescription())
   > "SetText (Text)"(self=Get Text_Creator())
   > "SetText (Text)"(self=Get Text_IslandName())
   > "SetText (Text)"(self=Get PublishedIslandCode())
  ]}
 MoveTabButtonSelection {nodes [
   n1 = MoveTabButtonSelection!() -> (then, Forward:bool=~) {is=title K2Node_FunctionEntry}
   > n2 = Branch(Condition=@n1)
   n3 = Get TabButtonGroup()
   > n4 = SelectNextButton!(execute, self:Common Button Group Base Object Reference=@n3, bAllowWrap:bool=false, bAllowSound:bool=true) -> (then)
   PrintString(execute=[@n4.then, @n5.then], InString="To String (Integer)"(InInt=GetSelectedButtonIndex(self=@n3) -> (ReturnValue=0)))
   n5 = SelectPreviousButton!(execute=@n2.else, self:Common Button Group Base Object Reference=@n3, bAllowWrap:bool=false, bAllowSound:bool=true) -> (then)
  ]}
 OnFocusReceived {nodes [
   OnFocusReceived!() -> (then, MyGeometry:Geometry Structure, InFocusEvent:Focus Event Structure) {is=title K2Node_FunctionEntry}
   > n1 = Sequence()
   > Branch(Condition=IsUsingGamepad!(self:Fort Global UIContext Object Reference=Get FortGlobalUIContext() -> (ReturnValue:Fort Global UIContext Object Reference) {is=title K2Node_GetSubsystem}) -> (ReturnValue:bool))
   > n2 = GetSelectedIndex!(execute:Exec, self:Object Reference=Get Switcher_Tabs()) -> (then:Exec, ReturnValue:int=0) {is=title K2Node_Message}
   > Branch(Condition="Equal (Integer)"(A=@n2, B=0))
   > CenterWidget(ToCenter=Get EditableText_IslandLink())
   > Return Node!(execute, ReturnValue:Event Reply Structure=Handled())
   Return Node(execute=@n1.then_1, ReturnValue=Unhandled())
  ]}
 Set Initialization Object {nodes [
   n1 = Set Initialization Object!() -> (then, Object:Object Reference) {is=title K2Node_FunctionEntry}
   > n2 = Set Init Object(Init Object:Object Reference=@n1) -> (Output_Get:Object Reference)
   > n3 = Cast To BGA_IslandPortal!(execute, Object=@n2) -> (then, CastFailed, AsBGA Island Portal:BGA Island Portal Object Reference) {is=title K2Node_DynamicCast}
   > SetTargetPortal!(execute, self:Fort Creative Island Link Screen Object Reference, Portal:Fort Athena Creative Portal Object Reference=@n3) -> (then)
   > Set Portal(Portal:BGA Island Portal Object Reference=@n3) -> (Output_Get:BGA Island Portal Object Reference)
  ]}
 SetSelectedIslandLink {nodes [
   n1 = SetSelectedIslandLink!() -> (then, NewSelectedIslandLink:Fort Creative Island Link Object Reference) {is=title K2Node_FunctionEntry}
   > n2 = Set SelectedIslandLink(SelectedIslandLink:Fort Creative Island Link Object Reference=@n1) -> (Output_Get:Fort Creative Island Link Object Reference)
   > SetLastSelectedIsland!(
    execute
    self:Fort Creative Island Link Screen Server Browser Object Reference
    InLastSelectedIsland:Fort Creative Island Link Object Reference=@n2
   ) -> (then)
   > n3 = Is Valid(InputObject=@n2)
   n4 = Get MatchMaking_Line_Description()
   > "SetText (Text)"(
    self=@n4
    InText=GetMatchmakingTypeDescription!(self:Fort Creative Island Link Screen Object Reference, IslandLink:Fort Creative Island Link Object Reference=@n2) -> (ReturnValue:Text)
   )
   > SetVisibility(self=@n4, InVisibility=Visible)
   n5 = IsPartyEligible!(self:Fort Creative Island Link Screen Object Reference, IslandLink:Fort Creative Island Link Object Reference=@n2) -> (ReturnValue:bool)
   n6 = "Equal (Integer)"(A=GetSelectedButtonIndex(self=Get TabButtonGroup()) -> (ReturnValue=0), B=0)
   > SetVisibility(
    self=Get Launch_Error() -> (Launch_Error:Lightning Violator Object Reference)
    InVisibility=Select!(Option 0:ESlateVisibility Enum=HitTestInvisible, Option 1:ESlateVisibility Enum=Collapsed, Index:bool=OR Boolean(A=@n6, B=@n5)) -> (ReturnValue:ESlateVisibility Enum)
   )
   n7 = Get Matchmaking_MaximumNumberOfPlayers!(self:Fort Creative Island Link Object Reference=@n2) -> (Matchmaking_MaximumNumberOfPlayers:int=0)
   > n8 = Branch(Condition="integer > integer"(A=@n7, B=0))
   n9 = Get MatchMaking_MaximumNumberOfPlayers()
   > "SetText (Text)"(self=@n9, InText=Format Text!(Format=Get MMS_MaxNumberOfPlayers_Text() -> (MMS_MaxNumberOfPlayers_Text:Text), 0:int=@n7) -> (Result))
   > SetVisibility(self=@n9, InVisibility=Visible)
   SetVisibility(execute=@n8.else, self=@n9, InVisibility=Collapsed)
   SetVisibility(execute=@n3.Is Not Valid, self=Get MatchMaking_Line_Description(), InVisibility=Collapsed)
   > SetVisibility(self=Get MatchMaking_MaximumNumberOfPlayers(), InVisibility=Collapsed)
  ]}
 UpdateIslandArtInternal {nodes [
   n1 = UpdateIslandArtInternal!() -> (then, inTexture:Texture 2DDynamic Object Reference) {is=title K2Node_FunctionEntry}
   > SetBrushFromTextureDynamic!(execute, self:Image Object Reference=Get Image_Island(), Texture:Texture 2DDynamic Object Reference=@n1, bMatchSize:bool=false) -> (then)
   > SetRenderOpacity(self=Get Image_Island(), InOpacity=1)
   > SetVisibility(self=Get ImageLoadingSpinner(), InVisibility=Collapsed)
   > SetVisibility(self=Get Border_IslandImage(), InVisibility=SelfHitTestInvisible)
  ]}
]
variables [
 DefaultArt {
  type=SlateBrush
  default {
   bIsDynamicallyLoaded=false
   drawAs=Image
   imageType=NoImage
   mirroring=NoMirror
   resourceName=None
   resourceObject=/Game/Playgrounds/Items/Portals/UI/TemplateFeaturedDefaultImage
   tiling=NoTile
   imageSize=<1024 512>
   margin=<ltrb 0 0 0 0>
   outlineSettings {
    bUseBrushTransparency=false
    roundingType=HalfHeightRadius
    width=0
    color {colorUseRule=UseColor_Specified, specifiedColor=<rgba 0 0 0 0>}
    cornerRadii=<0 0 0 1>
   }
   tintColor {colorUseRule=UseColor_Specified, specifiedColor=<rgba 1 1 1 0.3>}
   uVRegion {bIsValid=false, max=<0 0>, min=<0 0>}
  }
 }
 DefaultImage {type=boolean}
 Init Object {type=Object, default=null}
 LastValidateResult {type=EFortActivityValidationResult, default=NotFound}
 Portal {type=BGA_IslandPortal_C, default=null}
 SelectedIslandLink {type=FortCreativeIslandLink, default=null}
 SelectPlotSound {type=SoundBase, default=/Game/Sounds/Fort_UI_Sounds/FrontEnd/New_FrontEnd/Fortnite_UI_Options_Menu_01_Cue}
 TabButtonGroup {type=CommonButtonGroupLegacy, default=null}
 ViolatorColor_Error {type=LinearColor, default=<rgba 1 0.0296 0 1>}
 ViolatorColor_Success {type=LinearColor, default=<rgba 0 0.491 1 1>}
]
widgets {
 info {
  inheritedWidgetCount=9
  namedSlotCount=3
  parentClass=/Script/FortniteUI.FortCreativeIslandLinkScreenServerBrowser
  rootWidgetClass=&BackgroundBlur
  widgetCount=90
 }
 widgets [
  BackgroundBlur_0 {
   widgetClassPath=&BackgroundBlur
   widget=BackgroundBlur_0
   properties {
    blurStrength=25
    visibility=Visible
    lowQualityFallbackBrush {
     bIsDynamicallyLoaded=false
     drawAs=Image
     imageType=NoImage
     mirroring=NoMirror
     resourceName=None
     resourceObject=null
     tiling=NoTile
     imageSize=<0 0>
     margin=<ltrb 0 0 0 0>
     outlineSettings {
      bUseBrushTransparency=false
      roundingType=HalfHeightRadius
      width=0
      color {colorUseRule=UseColor_Specified, specifiedColor=<rgba 0 0 0 0>}
      cornerRadii=<0 0 0 0>
     }
     tintColor {colorUseRule=UseColor_Specified, specifiedColor=<rgba 1 1 1 0.8>}
     uVRegion {bIsValid=false, max=<0 0>, min=<0 0>}
    }
   }
  }
  SafeZone_0 {
   widgetClassPath=&SafeZone
   bIsVariable=true
   parent=BackgroundBlur_0
   widget=SafeZone_0
   slot {class=BackgroundBlurSlot, properties {padding=<ltrb 0 0 0 0>}}
  }
  Overlay_3 {widgetClassPath=&Overlay, parent=SafeZone_0, widget=Overlay_3, slot {class=SafeZoneSlot}}
  SizeBox_2 {
   widgetClassPath=&SizeBox
   parent=Overlay_3
   widget=SizeBox_2
   properties {
    bOverride_HeightOverride=true
    bOverride_WidthOverride=true
    heightOverride=720
    maxDesiredHeight=750
    minDesiredHeight=750
    minDesiredWidth=1200
    widthOverride=1600
   }
   slot {class=OverlaySlot, properties {horizontalAlignment=HAlign_Center, verticalAlignment=VAlign_Center}}
  }
  Overlay_1 {widgetClassPath=&Overlay, parent=SizeBox_2, widget=Overlay_1, slot {class=SizeBoxSlot}}
  Image_Shadow {
   widgetClassPath=&Image
   bIsVariable=true
   parent=Overlay_1
   widget=Image_Shadow
   properties {colorAndOpacity=<rgba 0.0052 0.0052 0.0056 0.4>, renderTransform {angle=0, scale=<1 1>, shear=<0 0>, translation=<12 12>}}
   slot {class=OverlaySlot, properties {horizontalAlignment=HAlign_Fill, verticalAlignment=VAlign_Fill}}
  }
  Overlay_7 {
   widgetClassPath=&Overlay
   parent=Overlay_1
   widget=Overlay_7
   slot {class=OverlaySlot, properties {horizontalAlignment=HAlign_Fill, verticalAlignment=VAlign_Fill}}
  }
  CommonBorder_0 {
   widgetClassPath=&CommonBorder
   parent=Overlay_7
   widget=CommonBorder_0
   properties {
    background {
     bIsDynamicallyLoaded=false
     drawAs=Box
     imageType=NoImage
     mirroring=NoMirror
     resourceName=None
     resourceObject=null
     tiling=NoTile
     imageSize=<32 32>
     margin=<ltrb 0 0 0 0>
     outlineSettings {
      bUseBrushTransparency=false
      roundingType=HalfHeightRadius
      width=0
      color {colorUseRule=UseColor_Specified, specifiedColor=<rgba 0 0 0 0>}
      cornerRadii=<0 0 0 0>
     }
     tintColor {colorUseRule=UseColor_Specified, specifiedColor=<rgba 0 0.0222 0.2874 1>}
     uVRegion {bIsValid=false, max=<0 0>, min=<0 0>}
    }
    padding=<ltrb 0 0 0 0>
   }
   slot {class=OverlaySlot, properties {horizontalAlignment=HAlign_Fill, verticalAlignment=VAlign_Fill}}
  }
  Overlay_6 {widgetClassPath=&Overlay, parent=CommonBorder_0, widget=Overlay_6, slot {class=BorderSlot, properties {padding=<ltrb 0 0 0 0>}}}
  VerticalBox_1 {
   widgetClassPath=&VerticalBox
   parent=Overlay_6
   widget=VerticalBox_1
   properties {navigation=&self:WidgetTree.VerticalBox_1.WidgetNavigation_0}
   slot {class=OverlaySlot, properties {horizontalAlignment=HAlign_Fill, verticalAlignment=VAlign_Fill}}
  }
  SizeBox_Header {
   widgetClassPath=&SizeBox
   parent=VerticalBox_1
   widget=SizeBox_Header
   properties {bOverride_HeightOverride=true, heightOverride=132}
   slot {class=VerticalBoxSlot, properties {padding=<ltrb 0 -2 0 0>}}
  }
  Overlay {widgetClassPath=&Overlay, parent=SizeBox_Header, widget=Overlay, slot {class=SizeBoxSlot}}
  Image {
   widgetClassPath=&Image
   parent=Overlay
   widget=Image
   properties {
    brush {
     bIsDynamicallyLoaded=false
     drawAs=Image
     imageType=NoImage
     mirroring=NoMirror
     resourceName=None
     resourceObject=/Game/Athena/UI/InGame/Creative/Art/M_UI_BoltGradientBorder_Creative_IslandSelectHeaderTop
     tiling=NoTile
     imageSize=<32 85.201>
     margin=<ltrb 0 0 0 0>
     outlineSettings {
      bUseBrushTransparency=false
      roundingType=HalfHeightRadius
      width=0
      color {colorUseRule=UseColor_Specified, specifiedColor=<rgba 0 0 0 0>}
      cornerRadii=<0 0 0 0>
     }
     tintColor {colorUseRule=UseColor_Specified, specifiedColor=<rgba 1 1 1 1>}
     uVRegion {bIsValid=false, max=<0 0>, min=<0 0>}
    }
    renderTransform {angle=0, scale=<1 -1>, shear=<0 0>, translation=<0 0>}
   }
   slot {class=OverlaySlot, properties {horizontalAlignment=HAlign_Fill, padding=<ltrb 0 -4 0 0>}}
  }
  Image_180 {
   widgetClassPath=&Image
   parent=Overlay
   widget=Image_180
   properties {
    brush {
     bIsDynamicallyLoaded=false
     drawAs=Image
     imageType=NoImage
     mirroring=NoMirror
     resourceName=None
     resourceObject=/Game/Athena/UI/InGame/Creative/Art/M_UI_BoltGradientBorder_Creative_IslandSelectHeader
     tiling=NoTile
     imageSize=<32 79.1332>
     margin=<ltrb 0 0 0 0>
     outlineSettings {
      bUseBrushTransparency=false
      roundingType=HalfHeightRadius
      width=0
      color {colorUseRule=UseColor_Specified, specifiedColor=<rgba 0 0 0 0>}
      cornerRadii=<0 0 0 0>
     }
     tintColor {colorUseRule=UseColor_Specified, specifiedColor=<rgba 1 1 1 1>}
     uVRegion {bIsValid=false, max=<0 0>, min=<0 0>}
    }
    renderTransform {angle=0, scale=<1 -1>, shear=<0 0>, translation=<0 0>}
   }
   slot {class=OverlaySlot, properties {horizontalAlignment=HAlign_Fill, verticalAlignment=VAlign_Bottom, padding=<ltrb 0 -3.2 0 0>}}
  }
  CommonTextBlock_2 {
   widgetClassPath=&CommonTextBlock
   bIsVariable=true
   parent=Overlay
   widget=CommonTextBlock_2
   properties {
    justification=Center
    text=CHOOSE DESTINATION
    colorAndOpacity {colorUseRule=UseColor_Specified, specifiedColor=<rgba 1 1 1 0.8>}
    font {
     bForceMonospaced=false
     bMaterialIsStencil=false
     fontMaterial=null
     fontObject=&Burbank
     letterSpacing=0
     monospacedWidth=1
     size=64
     skewAmount=0
     typefaceFontName=Burbank-Black
     outlineSettings {
      bApplyOutlineToDropShadows=false
      bMiteredCorners=false
      bSeparateFillAlpha=false
      outlineMaterial=null
      outlineSize=0
      outlineColor=<rgba 0 0 0 1>
     }
    }
    shadowOffset=<4.9266 5.2555>
   }
   slot {class=OverlaySlot, properties {horizontalAlignment=HAlign_Fill, verticalAlignment=VAlign_Center}}
  }
  Spacer_235 {widgetClassPath=&Spacer, parent=VerticalBox_1, widget=Spacer_235, properties {size=<1 24>}, slot {class=VerticalBoxSlot}}
  Overlay_0 {widgetClassPath=&Overlay, parent=VerticalBox_1, widget=Overlay_0, slot {class=VerticalBoxSlot}}
  HorizontalBox_144 {
   widgetClassPath=&HorizontalBox
   parent=Overlay_0
   widget=HorizontalBox_144
   slot {class=OverlaySlot, properties {horizontalAlignment=HAlign_Center, verticalAlignment=VAlign_Center}}
  }
  UserActionMenuInputButton_Left {
   widgetClassPath=/Game/UI/SocialPanel/UserActionMenuInputButton
   bIsVariable=true
   parent=HorizontalBox_144
   widget=UserActionMenuInputButton_Left
   properties {
    onButtonBaseClicked=(&self:WidgetTree.UserActionMenuInputButton_Left.HandleOnButtonClicked)
    onButtonBaseDoubleClicked=(&self:WidgetTree.UserActionMenuInputButton_Left.HandleOnButtonDoubleClicked)
    onButtonBaseHovered=(&self:WidgetTree.UserActionMenuInputButton_Left.HandleOnButtonHovered)
    onButtonBaseUnhovered=(&self:WidgetTree.UserActionMenuInputButton_Left.HandleOnButtonUnhovered)
    onSelectedChangedBase=(&self:WidgetTree.UserActionMenuInputButton_Left.HandleOnSelectedChanged)
    triggeringInputAction {dataTable=&InputActionDataTable, rowName=GenericLeftShoulder}
   }
   slot {class=HorizontalBoxSlot, properties {padding=<ltrb 40 0 5 0>}}
  }
  Tab_EnterCode {
   widgetClassPath=&IconTextButton
   bIsVariable=true
   parent=HorizontalBox_144
   widget=Tab_EnterCode
   properties {
    bHideInputAction=true
    bIsFocusable=false
    bSelectable=true
    bShouldSelectUponReceivingFocus=true
    buttonText=ENTER CODE
    minWidth=300
    onButtonBaseClicked=(&self:WidgetTree.Tab_EnterCode.HandleOnButtonClicked)
    onButtonBaseDoubleClicked=(&self:WidgetTree.Tab_EnterCode.HandleOnButtonDoubleClicked)
    onButtonBaseHovered=(&self:WidgetTree.Tab_EnterCode.HandleOnButtonHovered)
    onButtonBaseUnhovered=(&self:WidgetTree.Tab_EnterCode.HandleOnButtonUnhovered)
    onSelectedChangedBase=(&self:WidgetTree.Tab_EnterCode.HandleOnSelectedChanged)
    style=/FortUILibrary/Foundation/Buttons/ButtonStyles/Tab/Athena_ButtonStyle_ChallengesTab.Athena_ButtonStyle_ChallengesTab_C
    visibility=Visible
   }
   slot {class=HorizontalBoxSlot, properties {size {sizeRule=Fill, value=1}}}
  }
  RemoveSpinner_1 {
   widgetClassPath=&Image
   bIsVariable=true
   namedSlotHost=Tab_EnterCode
   widget=RemoveSpinner_1
   properties {
    visibility=Collapsed
    brush {
     bIsDynamicallyLoaded=false
     drawAs=Image
     imageType=NoImage
     mirroring=NoMirror
     resourceName=None
     resourceObject=/FortUILibrary/UIKit/Throbber/MI_UIKit_Monochrome_Throbber
     tiling=NoTile
     imageSize=<75 32>
     margin=<ltrb 0 0 0 0>
     outlineSettings {
      bUseBrushTransparency=false
      roundingType=HalfHeightRadius
      width=0
      color {colorUseRule=UseColor_Specified, specifiedColor=<rgba 0 0 0 0>}
      cornerRadii=<0 0 0 0>
     }
     tintColor {colorUseRule=UseColor_Specified, specifiedColor=<rgba 1 1 1 1>}
     uVRegion {bIsValid=false, max=<0 0>, min=<0 0>}
    }
   }
  }
  Tab_Recent {
   widgetClassPath=&IconTextButton
   bIsVariable=true
   parent=HorizontalBox_144
   widget=Tab_Recent
   properties {
    bHideInputAction=true
    bIsFocusable=false
    buttonText=RECENT ISLANDS
    minWidth=300
    onButtonBaseClicked=(&self:WidgetTree.Tab_Recent.HandleOnButtonClicked)
    onButtonBaseDoubleClicked=(&self:WidgetTree.Tab_Recent.HandleOnButtonDoubleClicked)
    onButtonBaseHovered=(&self:WidgetTree.Tab_Recent.HandleOnButtonHovered)
    onButtonBaseUnhovered=(&self:WidgetTree.Tab_Recent.HandleOnButtonUnhovered)
    onSelectedChangedBase=(&self:WidgetTree.Tab_Recent.HandleOnSelectedChanged)
    style=&Athena_ButtonStyle_ChallengesTab_C
    visibility=Visible
   }
   slot {class=HorizontalBoxSlot, properties {padding=<ltrb 8 0 0 0>, size {sizeRule=Fill, value=1}}}
  }
  RemoveSpinner_2 {
   widgetClassPath=&Image
   bIsVariable=true
   namedSlotHost=Tab_Recent
   widget=RemoveSpinner_2
   properties {
    visibility=Collapsed
    brush {
     bIsDynamicallyLoaded=false
     drawAs=Image
     imageType=NoImage
     mirroring=NoMirror
     resourceName=None
     resourceObject=&MI_UIKit_Monochrome_Throbber
     tiling=NoTile
     imageSize=<75 32>
     margin=<ltrb 0 0 0 0>
     outlineSettings {
      bUseBrushTransparency=false
      roundingType=HalfHeightRadius
      width=0
      color {colorUseRule=UseColor_Specified, specifiedColor=<rgba 0 0 0 0>}
      cornerRadii=<0 0 0 0>
     }
     tintColor {colorUseRule=UseColor_Specified, specifiedColor=<rgba 1 1 1 1>}
     uVRegion {bIsValid=false, max=<0 0>, min=<0 0>}
    }
   }
  }
  Tab_Favorites {
   widgetClassPath=&IconTextButton
   bIsVariable=true
   parent=HorizontalBox_144
   widget=Tab_Favorites
   properties {
    bHideInputAction=true
    bIsFocusable=false
    buttonText=FAVORITES
    minWidth=300
    onButtonBaseClicked=(&self:WidgetTree.Tab_Favorites.HandleOnButtonClicked)
    onButtonBaseDoubleClicked=(&self:WidgetTree.Tab_Favorites.HandleOnButtonDoubleClicked)
    onButtonBaseHovered=(&self:WidgetTree.Tab_Favorites.HandleOnButtonHovered)
    onButtonBaseUnhovered=(&self:WidgetTree.Tab_Favorites.HandleOnButtonUnhovered)
    onSelectedChangedBase=(&self:WidgetTree.Tab_Favorites.HandleOnSelectedChanged)
    style=&Athena_ButtonStyle_ChallengesTab_C
   }
   slot {class=HorizontalBoxSlot, properties {padding=<ltrb 8 0 0 0>, size {sizeRule=Fill, value=1}}}
  }
  RemoveSpinner {
   widgetClassPath=&Image
   bIsVariable=true
   namedSlotHost=Tab_Favorites
   widget=RemoveSpinner
   properties {
    visibility=Collapsed
    brush {
     bIsDynamicallyLoaded=false
     drawAs=Image
     imageType=NoImage
     mirroring=NoMirror
     resourceName=None
     resourceObject=&MI_UIKit_Monochrome_Throbber
     tiling=NoTile
     imageSize=<75 32>
     margin=<ltrb 0 0 0 0>
     outlineSettings {
      bUseBrushTransparency=false
      roundingType=HalfHeightRadius
      width=0
      color {colorUseRule=UseColor_Specified, specifiedColor=<rgba 0 0 0 0>}
      cornerRadii=<0 0 0 0>
     }
     tintColor {colorUseRule=UseColor_Specified, specifiedColor=<rgba 1 1 1 1>}
     uVRegion {bIsValid=false, max=<0 0>, min=<0 0>}
    }
   }
  }
  UserActionMenuInputButton_Right {
   widgetClassPath=&UserActionMenuInputButton
   bIsVariable=true
   parent=HorizontalBox_144
   widget=UserActionMenuInputButton_Right
   properties {
    onButtonBaseClicked=(&self:WidgetTree.UserActionMenuInputButton_Right.HandleOnButtonClicked)
    onButtonBaseDoubleClicked=(&self:WidgetTree.UserActionMenuInputButton_Right.HandleOnButtonDoubleClicked)
    onButtonBaseHovered=(&self:WidgetTree.UserActionMenuInputButton_Right.HandleOnButtonHovered)
    onButtonBaseUnhovered=(&self:WidgetTree.UserActionMenuInputButton_Right.HandleOnButtonUnhovered)
    onSelectedChangedBase=(&self:WidgetTree.UserActionMenuInputButton_Right.HandleOnSelectedChanged)
    triggeringInputAction {dataTable=&InputActionDataTable, rowName=GenericRightShoulder}
   }
   slot {class=HorizontalBoxSlot, properties {padding=<ltrb 5 0 0 0>}}
  }
  Spacer_780 {widgetClassPath=&Spacer, parent=VerticalBox_1, widget=Spacer_780, properties {size=<1 12>}, slot {class=VerticalBoxSlot}}
  Overlay_8 {
   widgetClassPath=&Overlay
   parent=VerticalBox_1
   widget=Overlay_8
   properties {clipping=ClipToBounds}
   slot {class=VerticalBoxSlot, properties {size {sizeRule=Fill, value=1}}}
  }
  SizeBox_5 {
   widgetClassPath=&SizeBox
   parent=Overlay_8
   widget=SizeBox_5
   properties {bOverride_WidthOverride=true, widthOverride=720}
   slot {class=OverlaySlot, properties {horizontalAlignment=HAlign_Right, verticalAlignment=VAlign_Fill}}
  }
  Overlay_4 {widgetClassPath=&Overlay, parent=SizeBox_5, widget=Overlay_4, slot {class=SizeBoxSlot}}
  Image_rightside {
   widgetClassPath=&Image
   parent=Overlay_4
   widget=Image_rightside
   properties {
    brush {
     bIsDynamicallyLoaded=false
     drawAs=NoDrawType
     imageType=NoImage
     mirroring=NoMirror
     resourceName=None
     resourceObject=/Game/Athena/UI/InGame/Creative/Art/M_CreativeCreateIslandRightSideBacking
     tiling=NoTile
     imageSize=<999.9958 32>
     margin=<ltrb 0 0 0 0>
     outlineSettings {
      bUseBrushTransparency=false
      roundingType=HalfHeightRadius
      width=0
      color {colorUseRule=UseColor_Specified, specifiedColor=<rgba 0 0 0 0>}
      cornerRadii=<0 0 0 0>
     }
     tintColor {colorUseRule=UseColor_Specified, specifiedColor=<rgba 1 1 1 1>}
     uVRegion {bIsValid=false, max=<0 0>, min=<0 0>}
    }
    colorAndOpacity=<rgba 0 0 0 0.2>
   }
   slot {class=OverlaySlot, properties {horizontalAlignment=HAlign_Right, verticalAlignment=VAlign_Fill}}
  }
  ScaleBox_0 {
   widgetClassPath=&ScaleBox
   parent=Overlay_4
   widget=ScaleBox_0
   properties {stretch=ScaleToFitX}
   slot {class=OverlaySlot, properties {horizontalAlignment=HAlign_Fill}}
  }
  Image_Island {
   widgetClassPath=&Image
   bIsVariable=true
   parent=ScaleBox_0
   widget=Image_Island
   properties {brush {
     bIsDynamicallyLoaded=false
     drawAs=Image
     imageType=NoImage
     mirroring=NoMirror
     resourceName=None
     resourceObject=&TemplateFeaturedDefaultImage
     tiling=NoTile
     imageSize=<512 288>
     margin=<ltrb 0 0 0 0>
     outlineSettings {
      bUseBrushTransparency=false
      roundingType=HalfHeightRadius
      width=0
      color {colorUseRule=UseColor_Specified, specifiedColor=<rgba 0 0 0 0>}
      cornerRadii=<0 0 0 0>
     }
     tintColor {colorUseRule=UseColor_Specified, specifiedColor=<rgba 1 1 1 1>}
     uVRegion {bIsValid=false, max=<0 0>, min=<0 0>}
    }}
   slot {class=ScaleBoxSlot, properties {verticalAlignment=VAlign_Top}}
  }
  ImageLoadingSpinner {
   widgetClassPath=&Image
   bIsVariable=true
   parent=Overlay_4
   widget=ImageLoadingSpinner
   properties {
    visibility=Collapsed
    brush {
     bIsDynamicallyLoaded=false
     drawAs=Image
     imageType=NoImage
     mirroring=NoMirror
     resourceName=None
     resourceObject=&MI_UIKit_Monochrome_Throbber
     tiling=NoTile
     imageSize=<64 64>
     margin=<ltrb 0 0 0 0>
     outlineSettings {
      bUseBrushTransparency=false
      roundingType=HalfHeightRadius
      width=0
      color {colorUseRule=UseColor_Specified, specifiedColor=<rgba 0 0 0 0>}
      cornerRadii=<0 0 0 0>
     }
     tintColor {colorUseRule=UseColor_Specified, specifiedColor=<rgba 1 1 1 1>}
     uVRegion {bIsValid=false, max=<0 0>, min=<0 0>}
    }
   }
   slot {class=OverlaySlot, properties {horizontalAlignment=HAlign_Center, verticalAlignment=VAlign_Center}}
  }
  Overlay_11 {
   widgetClassPath=&Overlay
   parent=Overlay_4
   widget=Overlay_11
   slot {class=OverlaySlot, properties {horizontalAlignment=HAlign_Fill, verticalAlignment=VAlign_Bottom}}
  }
  Image_2 {
   widgetClassPath=&Image
   parent=Overlay_11
   widget=Image_2
   properties {brush {
     bIsDynamicallyLoaded=false
     drawAs=Image
     imageType=NoImage
     mirroring=NoMirror
     resourceName=None
     resourceObject=/Game/Athena/UI/InGame/Creative/Art/M_UI_BoltGradientBorder_Creative_IslandSelectDetailsBacking
     tiling=NoTile
     imageSize=<32 32>
     margin=<ltrb 0 0 0 0>
     outlineSettings {
      bUseBrushTransparency=false
      roundingType=HalfHeightRadius
      width=0
      color {colorUseRule=UseColor_Specified, specifiedColor=<rgba 0 0 0 0>}
      cornerRadii=<0 0 0 0>
     }
     tintColor {colorUseRule=UseColor_Specified, specifiedColor=<rgba 1 1 1 1>}
     uVRegion {bIsValid=false, max=<0 0>, min=<0 0>}
    }}
   slot {class=OverlaySlot, properties {horizontalAlignment=HAlign_Fill, verticalAlignment=VAlign_Fill, padding=<ltrb 0 -32 0 0>}}
  }
  SizeBox_3 {
   widgetClassPath=&SizeBox
   parent=Overlay_11
   widget=SizeBox_3
   properties {bOverride_MaxDesiredHeight=true, heightOverride=300, maxDesiredHeight=300}
   slot {class=OverlaySlot, properties {horizontalAlignment=HAlign_Fill, verticalAlignment=VAlign_Bottom, padding=<ltrb 42 0 0 102>}}
  }
  ScrollBox_Details {
   widgetClassPath=&ScrollBox
   bInherited=true
   bIsVariable=true
   parent=SizeBox_3
   widget=ScrollBox_Details
   slot {class=SizeBoxSlot}
  }
  HorizontalBox_5 {widgetClassPath=&HorizontalBox, parent=ScrollBox_Details, widget=HorizontalBox_5, slot {class=ScrollBoxSlot}}
  Text_CreatorLabel {
   widgetClassPath=&CommonTextBlock
   bIsVariable=true
   parent=HorizontalBox_5
   widget=Text_CreatorLabel
   properties {
    text=CREATED BY
    textTransformPolicy=ToUpper
    colorAndOpacity {colorUseRule=UseColor_Specified, specifiedColor=<rgba 0.5271 1 0.8228 1>}
    font {
     bForceMonospaced=false
     bMaterialIsStencil=false
     fontMaterial=null
     fontObject=&Burbank
     letterSpacing=0
     monospacedWidth=1
     size=18
     skewAmount=0
     typefaceFontName=BurbankRegular-Black
     outlineSettings {
      bApplyOutlineToDropShadows=false
      bMiteredCorners=false
      bSeparateFillAlpha=false
      outlineMaterial=null
      outlineSize=0
      outlineColor=<rgba 0 0 0 1>
     }
    }
    renderTransform {angle=0, scale=<1 1>, shear=<-12 0>, translation=<0 0>}
    shadowOffset=<0 0>
   }
   slot {class=HorizontalBoxSlot, properties {padding=<ltrb 0 0 8 0>}}
  }
  Text_Creator {
   widgetClassPath=&CommonTextBlock
   bIsVariable=true
   parent=HorizontalBox_5
   widget=Text_Creator
   properties {
    autoWrapText=true
    text="[CREATOR]"
    textTransformPolicy=ToUpper
    colorAndOpacity {colorUseRule=UseColor_Specified, specifiedColor=<rgba 0 1 0.6654 1>}
    font {
     bForceMonospaced=false
     bMaterialIsStencil=false
     fontMaterial=null
     fontObject=&Burbank
     letterSpacing=0
     monospacedWidth=1
     size=18
     skewAmount=0
     typefaceFontName=BurbankRegular-Black
     outlineSettings {
      bApplyOutlineToDropShadows=false
      bMiteredCorners=false
      bSeparateFillAlpha=false
      outlineMaterial=null
      outlineSize=0
      outlineColor=<rgba 0 0 0 1>
     }
    }
    renderTransform {angle=0, scale=<1 1>, shear=<-12 0>, translation=<0 0>}
    shadowOffset=<0 0>
   }
   slot {class=HorizontalBoxSlot, properties {size {sizeRule=Fill, value=1}}}
  }
  PublishedIslandCode {
   widgetClassPath=&CommonTextBlock
   bIsVariable=true
   parent=ScrollBox_Details
   widget=PublishedIslandCode
   properties {
    text=8888-8888-8888
    textTransformPolicy=ToUpper
    colorAndOpacity {colorUseRule=UseColor_Specified, specifiedColor=<rgba 0.502 0.8314 1 1>}
    font {
     bForceMonospaced=false
     bMaterialIsStencil=false
     fontMaterial=null
     fontObject=&Burbank
     letterSpacing=0
     monospacedWidth=1
     size=18
     skewAmount=0
     typefaceFontName=Burbank-Black
     outlineSettings {
      bApplyOutlineToDropShadows=false
      bMiteredCorners=false
      bSeparateFillAlpha=false
      outlineMaterial=null
      outlineSize=0
      outlineColor=<rgba 0 0 0 1>
     }
    }
    shadowOffset=<4.9266 5.2555>
   }
   slot {class=ScrollBoxSlot, properties {verticalAlignment=VAlign_Center}}
  }
  MatchMaking_MaximumNumberOfPlayers {
   widgetClassPath=&CommonTextBlock
   bIsVariable=true
   parent=ScrollBox_Details
   widget=MatchMaking_MaximumNumberOfPlayers
   properties {
    text="Max Players: {0}"
    textTransformPolicy=ToUpper
    colorAndOpacity {colorUseRule=UseColor_Specified, specifiedColor=<rgba 0.502 0.8314 1 1>}
    font {
     bForceMonospaced=false
     bMaterialIsStencil=false
     fontMaterial=null
     fontObject=&Burbank
     letterSpacing=0
     monospacedWidth=1
     size=18
     skewAmount=0
     typefaceFontName=Burbank-Black
     outlineSettings {
      bApplyOutlineToDropShadows=false
      bMiteredCorners=false
      bSeparateFillAlpha=false
      outlineMaterial=null
      outlineSize=0
      outlineColor=<rgba 0 0 0 1>
     }
    }
    shadowOffset=<4.9266 5.2555>
   }
   slot {class=ScrollBoxSlot, properties {verticalAlignment=VAlign_Center}}
  }
  MatchMaking_Line_Description {
   widgetClassPath=&CommonTextBlock
   bIsVariable=true
   parent=ScrollBox_Details
   widget=MatchMaking_Line_Description
   properties {
    text=Play with your party and others
    textTransformPolicy=ToUpper
    colorAndOpacity {colorUseRule=UseColor_Specified, specifiedColor=<rgba 0.502 0.8314 1 1>}
    font {
     bForceMonospaced=false
     bMaterialIsStencil=false
     fontMaterial=null
     fontObject=&Burbank
     letterSpacing=0
     monospacedWidth=1
     size=18
     skewAmount=0
     typefaceFontName=Burbank-Black
     outlineSettings {
      bApplyOutlineToDropShadows=false
      bMiteredCorners=false
      bSeparateFillAlpha=false
      outlineMaterial=null
      outlineSize=0
      outlineColor=<rgba 0 0 0 1>
     }
    }
    shadowOffset=<4.9266 5.2555>
   }
   slot {class=ScrollBoxSlot, properties {verticalAlignment=VAlign_Center}}
  }
  Text_IslandName {
   widgetClassPath=&CommonTextBlock
   bIsVariable=true
   parent=ScrollBox_Details
   widget=Text_IslandName
   properties {
    autoWrapText=true
    text="[Name]"
    textTransformPolicy=ToUpper
    colorAndOpacity {colorUseRule=UseColor_Specified, specifiedColor=<rgba 0.2195 0.6584 1 1>}
    font {
     bForceMonospaced=false
     bMaterialIsStencil=false
     fontMaterial=/Game/Athena/UI/Challenges/Art/M_UI_Challenge_DynamicColor_MissionHover
     fontObject=&Burbank
     letterSpacing=0
     monospacedWidth=1
     size=24
     skewAmount=0
     typefaceFontName=BurbankRegular-Black
     outlineSettings {
      bApplyOutlineToDropShadows=false
      bMiteredCorners=false
      bSeparateFillAlpha=false
      outlineMaterial=null
      outlineSize=0
      outlineColor=<rgba 0 0 0 1>
     }
    }
    shadowOffset=<0 0>
   }
   slot {class=ScrollBoxSlot, properties {verticalAlignment=VAlign_Top}}
  }
  Text_IslandDescription {
   widgetClassPath=&CommonTextBlock
   bIsVariable=true
   parent=ScrollBox_Details
   widget=Text_IslandDescription
   properties {
    autoWrapText=true
    text="[Description] "
    textTransformPolicy=ToUpper
    colorAndOpacity {colorUseRule=UseColor_Specified, specifiedColor=<rgba 0.0331 0.3968 1 1>}
    font {
     bForceMonospaced=false
     bMaterialIsStencil=false
     fontMaterial=null
     fontObject=&BurbankSmall
     letterSpacing=0
     monospacedWidth=1
     size=18
     skewAmount=0
     typefaceFontName=Black
     outlineSettings {
      bApplyOutlineToDropShadows=false
      bMiteredCorners=false
      bSeparateFillAlpha=false
      outlineMaterial=null
      outlineSize=0
      outlineColor=<rgba 0 0 0 1>
     }
    }
    shadowOffset=<0 0>
   }
   slot {class=ScrollBoxSlot, properties {verticalAlignment=VAlign_Top}}
  }
  TagDisplay {
   widgetClassPath=/Game/Athena/UI/InGame/Creative/Cr_TagDisplay
   bIsVariable=true
   parent=ScrollBox_Details
   widget=TagDisplay
   slot {class=ScrollBoxSlot, properties {padding=<ltrb 0 4 0 0>}}
  }
  Image_leftside {
   widgetClassPath=&Image
   parent=Overlay_8
   widget=Image_leftside
   properties {
    brush {
     bIsDynamicallyLoaded=false
     drawAs=Image
     imageType=NoImage
     mirroring=NoMirror
     resourceName=None
     resourceObject=&M_CreativeCreateIslandRightSideBacking
     tiling=NoTile
     imageSize=<1020 589.9893>
     margin=<ltrb 0 0 0 0>
     outlineSettings {
      bUseBrushTransparency=false
      roundingType=HalfHeightRadius
      width=0
      color {colorUseRule=UseColor_Specified, specifiedColor=<rgba 0 0 0 0>}
      cornerRadii=<0 0 0 0>
     }
     tintColor {colorUseRule=UseColor_Specified, specifiedColor=<rgba 1 1 1 1>}
     uVRegion {bIsValid=false, max=<0 0>, min=<0 0>}
    }
    colorAndOpacity=<rgba 0 0.0222 0.2874 1>
    renderTransform {angle=0, scale=<-1 -1>, shear=<0 0>, translation=<0 0>}
   }
   slot {class=OverlaySlot, properties {verticalAlignment=VAlign_Fill, padding=<ltrb 0 0 0 -20>}}
  }
  Switcher_Tabs {
   widgetClassPath=&CommonWidgetSwitcherLegacy
   bInherited=true
   bIsVariable=true
   parent=Overlay_8
   widget=Switcher_Tabs
   properties {transitionDuration=0}
   slot {class=OverlaySlot, properties {horizontalAlignment=HAlign_Fill, verticalAlignment=VAlign_Fill, padding=<ltrb 0 0 700 120>}}
  }
  Overlay_EnterCode {widgetClassPath=&Overlay, parent=Switcher_Tabs, widget=Overlay_EnterCode, slot {class=WidgetSwitcherSlot}}
  VerticalBox_0 {
   widgetClassPath=&VerticalBox
   parent=Overlay_EnterCode
   widget=VerticalBox_0
   slot {class=OverlaySlot, properties {horizontalAlignment=HAlign_Fill, verticalAlignment=VAlign_Fill}}
  }
  Spacer_947 {widgetClassPath=&Spacer, parent=VerticalBox_0, widget=Spacer_947, properties {size=<1 50>}, slot {class=VerticalBoxSlot}}
  Text_Islands {
   widgetClassPath=&CommonTextBlock
   parent=VerticalBox_0
   widget=Text_Islands
   properties {
    justification=Center
    text=ENTER AN ISLAND CODE
    textTransformPolicy=ToUpper
    font {
     bForceMonospaced=false
     bMaterialIsStencil=false
     fontMaterial=&M_UI_Challenge_DynamicColor_MissionHover
     fontObject=&Burbank
     letterSpacing=0
     monospacedWidth=1
     size=32
     skewAmount=0
     typefaceFontName=BurbankRegular-Black
     outlineSettings {
      bApplyOutlineToDropShadows=false
      bMiteredCorners=false
      bSeparateFillAlpha=false
      outlineMaterial=null
      outlineSize=0
      outlineColor=<rgba 0 0 0 1>
     }
    }
    renderTransform {angle=0, scale=<1 1>, shear=<-12 0>, translation=<0 0>}
    shadowOffset=<0 0>
   }
   slot {class=VerticalBoxSlot, properties {horizontalAlignment=HAlign_Center, verticalAlignment=VAlign_Center}}
  }
  Spacer_1215 {widgetClassPath=&Spacer, parent=VerticalBox_0, widget=Spacer_1215, properties {size=<1 42>}, slot {class=VerticalBoxSlot}}
  SizeBox_0 {
   widgetClassPath=&SizeBox
   parent=VerticalBox_0
   widget=SizeBox_0
   properties {bOverride_MinDesiredWidth=true, minDesiredWidth=524, widthOverride=600}
   slot {class=VerticalBoxSlot, properties {horizontalAlignment=HAlign_Center, padding=<ltrb 50 0 50 0>}}
  }
  Overlay_Code {widgetClassPath=&Overlay, bIsVariable=true, parent=SizeBox_0, widget=Overlay_Code, slot {class=SizeBoxSlot}}
  Image_181 {
   widgetClassPath=&Image
   parent=Overlay_Code
   widget=Image_181
   properties {colorAndOpacity=<rgba 1 1 1 0.2>, renderTransform {angle=0, scale=<1 1>, shear=<-4 0>, translation=<0 0>}}
   slot {class=OverlaySlot, properties {horizontalAlignment=HAlign_Fill, verticalAlignment=VAlign_Fill}}
  }
  EditableText_IslandLink {
   widgetClassPath=/Script/UMG.EditableText
   bInherited=true
   bIsVariable=true
   parent=Overlay_Code
   widget=EditableText_IslandLink
   properties {
    clearKeyboardFocusOnCommit=false
    hintText=blah
    widgetStyle {
     backgroundImageComposing {
      bIsDynamicallyLoaded=false
      drawAs=Image
      imageType=FullColor
      mirroring=NoMirror
      resourceName=../../../Engine/Content/Slate/Old/HyperlinkDotted.png
      resourceObject=null
      tiling=NoTile
      imageSize=<0 0>
      margin=<ltrb 0 0 0 0.1875>
      outlineSettings {
       bUseBrushTransparency=false
       roundingType=HalfHeightRadius
       width=0
       color {colorUseRule=UseColor_Specified, specifiedColor=<rgba 0 0 0 0>}
       cornerRadii=<0 0 0 0>
      }
      tintColor {colorUseRule=UseColor_Specified, specifiedColor=<rgba 1 1 1 1>}
      uVRegion {bIsValid=false, max=<0 0>, min=<0 0>}
     }
     backgroundImageSelected {
      bIsDynamicallyLoaded=false
      drawAs=Box
      imageType=FullColor
      mirroring=NoMirror
      resourceName=../../../Engine/Content/Slate/Common/EditableTextSelectionBackground.png
      resourceObject=null
      tiling=NoTile
      imageSize=<0 0>
      margin=<ltrb 0.25 0.25 0.25 0.25>
      outlineSettings {
       bUseBrushTransparency=false
       roundingType=HalfHeightRadius
       width=0
       color {colorUseRule=UseColor_Specified, specifiedColor=<rgba 0 0 0 0>}
       cornerRadii=<0 0 0 0>
      }
      tintColor {colorUseRule=UseColor_Specified, specifiedColor=<rgba 0 0.0075 0.0423 1>}
      uVRegion {bIsValid=false, max=<0 0>, min=<0 0>}
     }
     caretImage {
      bIsDynamicallyLoaded=false
      drawAs=Image
      imageType=FullColor
      mirroring=NoMirror
      resourceName=../../../Engine/Content/Slate/Old/White.png
      resourceObject=null
      tiling=NoTile
      imageSize=<16 16>
      margin=<ltrb 0 0 0 0>
      outlineSettings {
       bUseBrushTransparency=false
       roundingType=HalfHeightRadius
       width=0
       color {colorUseRule=UseColor_Specified, specifiedColor=<rgba 0 0 0 0>}
       cornerRadii=<0 0 0 0>
      }
      tintColor {colorUseRule=UseColor_Specified, specifiedColor=<rgba 1 1 1 1>}
      uVRegion {bIsValid=false, max=<0 0>, min=<0 0>}
     }
     colorAndOpacity {colorUseRule=UseColor_Specified, specifiedColor=<rgba 0 0.491 1 1>}
     font {
      bForceMonospaced=false
      bMaterialIsStencil=false
      fontMaterial=null
      fontObject=&Burbank
      letterSpacing=0
      monospacedWidth=1
      size=58
      skewAmount=0
      typefaceFontName=Burbank-Black
      outlineSettings {
       bApplyOutlineToDropShadows=false
       bMiteredCorners=false
       bSeparateFillAlpha=false
       outlineMaterial=null
       outlineSize=0
       outlineColor=<rgba 0 0 0 1>
      }
     }
     hintColorAndOpacity {colorUseRule=UseColor_Foreground, specifiedColor=<rgba 1 0 1 1>}
     selectedBackgroundColor {colorUseRule=UseColor_Foreground, specifiedColor=<rgba 1 0 1 1>}
    }
   }
   slot {class=OverlaySlot, properties {horizontalAlignment=HAlign_Center, verticalAlignment=VAlign_Bottom, padding=<ltrb 12 16 12 0>}}
  }
  Text_WebURL {
   widgetClassPath=&CommonTextBlock
   bIsVariable=true
   parent=Overlay_Code
   widget=Text_WebURL
   properties {
    justification=Center
    text=blah
    visibility=HitTestInvisible
    colorAndOpacity {colorUseRule=UseColor_Specified, specifiedColor=<rgba 0.2805 0.3364 0.4852 1>}
    font {
     bForceMonospaced=false
     bMaterialIsStencil=false
     fontMaterial=null
     fontObject=&BurbankSmall
     letterSpacing=0
     monospacedWidth=1
     size=16
     skewAmount=0
     typefaceFontName=Black
     outlineSettings {
      bApplyOutlineToDropShadows=false
      bMiteredCorners=false
      bSeparateFillAlpha=false
      outlineMaterial=null
      outlineSize=0
      outlineColor=<rgba 0 0 0 1>
     }
    }
    shadowOffset=<0 0>
   }
   slot {class=OverlaySlot, properties {horizontalAlignment=HAlign_Center}}
  }
  Overlay_Violator {
   widgetClassPath=&Overlay
   bIsVariable=true
   parent=VerticalBox_0
   widget=Overlay_Violator
   properties {visibility=Collapsed}
   slot {class=VerticalBoxSlot, properties {verticalAlignment=VAlign_Bottom, padding=<ltrb 0 12 0 0>}}
  }
  LightningViolator {
   widgetClassPath=/Game/UI/Foundation/Bangs/LightningViolator
   bIsVariable=true
   parent=Overlay_Violator
   widget=LightningViolator
   properties {showOnTop=true, tailHorizontalAlignment=HAlign_Center}
   slot {class=OverlaySlot, properties {horizontalAlignment=HAlign_Center, verticalAlignment=VAlign_Center}}
  }
  SizeBox_7 {
   widgetClassPath=&SizeBox
   parent=Switcher_Tabs
   widget=SizeBox_7
   properties {bOverride_WidthOverride=true, widthOverride=860}
   slot {class=WidgetSwitcherSlot, properties {horizontalAlignment=HAlign_Left, padding=<ltrb 24 0 0 0>}}
  }
  Overlay_List {widgetClassPath=&Overlay, parent=SizeBox_7, widget=Overlay_List, slot {class=SizeBoxSlot}}
  Switcher_IslandsList {
   widgetClassPath=&CommonActivatableWidgetSwitcher
   bIsVariable=true
   parent=Overlay_List
   widget=Switcher_IslandsList
   properties {transitionDuration=0}
   slot {class=OverlaySlot, properties {horizontalAlignment=HAlign_Fill, verticalAlignment=VAlign_Fill, padding=<ltrb 0 10 0 0>}}
  }
  IslandListContent {
   widgetClassPath=&Overlay
   bIsVariable=true
   parent=Switcher_IslandsList
   widget=IslandListContent
   slot {class=WidgetSwitcherSlot}
  }
  ListView_IslandLinks {
   widgetClassPath=&CommonListView
   bInherited=true
   bIsVariable=true
   parent=IslandListContent
   widget=ListView_IslandLinks
   properties {
    entryWidgetClass=/Game/UI/Creative/CreativeIslandSetPortalEntry.CreativeIslandSetPortalEntry_C
    horizontalEntrySpacing=4
    navigation=&self:WidgetTree.ListView_IslandLinks.WidgetNavigation_0
    verticalEntrySpacing=4
   }
   slot {class=OverlaySlot, properties {horizontalAlignment=HAlign_Fill, verticalAlignment=VAlign_Fill}}
  }
  NoIslandsContent {widgetClassPath=&Overlay, bIsVariable=true, parent=Switcher_IslandsList, widget=NoIslandsContent, slot {class=WidgetSwitcherSlot}}
  Text_NoIslands {
   widgetClassPath=&CommonTextBlock
   bIsVariable=true
   parent=NoIslandsContent
   widget=Text_NoIslands
   properties {
    autoWrapText=true
    style=/FortUILibrary/Styles/Typography/Foundation/TextStyle-DecoBase-M.TextStyle-DecoBase-M_C
    text="Use the ENTER CODE tab to type in a code for an island to visit.\r\nPreviously entered codes will show up here as favorites.\r\n\r\n*Island Sharing is in limited access, but soon you'll be able to generate your own island codes to share with friends (and rivals)!*"
    font {
     bForceMonospaced=false
     bMaterialIsStencil=false
     fontMaterial=null
     fontObject=&BurbankSmall
     letterSpacing=0
     monospacedWidth=1
     size=24
     skewAmount=0
     typefaceFontName=Bold
     outlineSettings {
      bApplyOutlineToDropShadows=false
      bMiteredCorners=false
      bSeparateFillAlpha=false
      outlineMaterial=null
      outlineSize=0
      outlineColor=<rgba 0 0 0 1>
     }
    }
    shadowOffset=<0 0>
   }
   slot {class=OverlaySlot, properties {padding=<ltrb 0 50 0 0>}}
  }
  NoFavorites {widgetClassPath=&Overlay, bIsVariable=true, parent=Switcher_IslandsList, widget=NoFavorites, slot {class=WidgetSwitcherSlot}}
  Text_NoIslands_1 {
   widgetClassPath=&CommonTextBlock
   bIsVariable=true
   parent=NoFavorites
   widget=Text_NoIslands_1
   properties {
    autoWrapText=true
    renderOpacity=0.7
    style=&TextStyle-DecoBase-M_C
    text=No favorites marked
    font {
     bForceMonospaced=false
     bMaterialIsStencil=false
     fontMaterial=null
     fontObject=&BurbankSmall
     letterSpacing=0
     monospacedWidth=1
     size=24
     skewAmount=0
     typefaceFontName=Bold
     outlineSettings {
      bApplyOutlineToDropShadows=false
      bMiteredCorners=false
      bSeparateFillAlpha=false
      outlineMaterial=null
      outlineSize=0
      outlineColor=<rgba 0 0 0 1>
     }
    }
    renderTransform {angle=0, scale=<1 1>, shear=<-12 0>, translation=<0 0>}
    shadowOffset=<0 0>
   }
   slot {class=OverlaySlot, properties {horizontalAlignment=HAlign_Fill, padding=<ltrb 32 32 32 32>}}
  }
  Overlay_2 {
   widgetClassPath=&Overlay
   parent=Overlay_8
   widget=Overlay_2
   slot {class=OverlaySlot, properties {horizontalAlignment=HAlign_Fill, verticalAlignment=VAlign_Bottom}}
  }
  Image_1 {
   widgetClassPath=&Image
   bIsVariable=true
   parent=Overlay_2
   widget=Image_1
   properties {
    brush {
     bIsDynamicallyLoaded=false
     drawAs=Image
     imageType=NoImage
     mirroring=NoMirror
     resourceName=None
     resourceObject=/Game/Athena/UI/InGame/Creative/Art/M_UI_BoltGradientBorder_Creative_IslandSelectFooter
     tiling=NoTile
     imageSize=<32 64>
     margin=<ltrb 0 0 0 0>
     outlineSettings {
      bUseBrushTransparency=false
      roundingType=HalfHeightRadius
      width=0
      color {colorUseRule=UseColor_Specified, specifiedColor=<rgba 0 0 0 0>}
      cornerRadii=<0 0 0 0>
     }
     tintColor {colorUseRule=UseColor_Specified, specifiedColor=<rgba 1 1 1 1>}
     uVRegion {bIsValid=false, max=<0 0>, min=<0 0>}
    }
    colorAndOpacity=<rgba 0.0052 0.0578 0.3916 1>
   }
   slot {class=OverlaySlot, properties {horizontalAlignment=HAlign_Fill, verticalAlignment=VAlign_Bottom, padding=<ltrb -4 0 -4 -2>}}
  }
  HorizontalBox_0 {
   widgetClassPath=&HorizontalBox
   parent=Overlay_2
   widget=HorizontalBox_0
   slot {class=OverlaySlot, properties {horizontalAlignment=HAlign_Fill, verticalAlignment=VAlign_Bottom, padding=<ltrb 24 0 24 0>}}
  }
  HBox_ListActions {
   widgetClassPath=&HorizontalBox
   bIsVariable=true
   parent=HorizontalBox_0
   widget=HBox_ListActions
   properties {visibility=Collapsed}
   slot {class=HorizontalBoxSlot}
  }
  Button_Favorite {
   widgetClassPath=&IconTextButton
   bIsVariable=true
   parent=HBox_ListActions
   widget=Button_Favorite
   properties {
    bIsFocusable=false
    buttonText=TOGGLE FAVORITE
    contentAlignment=HAlign_Center
    onButtonBaseClicked=(&self:WidgetTree.Button_Favorite.HandleOnButtonClicked)
    onButtonBaseDoubleClicked=(&self:WidgetTree.Button_Favorite.HandleOnButtonDoubleClicked)
    onButtonBaseHovered=(&self:WidgetTree.Button_Favorite.HandleOnButtonHovered)
    onButtonBaseUnhovered=(&self:WidgetTree.Button_Favorite.HandleOnButtonUnhovered)
    onSelectedChangedBase=(&self:WidgetTree.Button_Favorite.HandleOnSelectedChanged)
    style=/FortUILibrary/Foundation/Buttons/ButtonStyles/Athena/Athena_ButtonStyle_AngledBlueMenuButton.Athena_ButtonStyle_AngledBlueMenuButton_C
    iconBrush {
     bIsDynamicallyLoaded=false
     drawAs=NoDrawType
     imageType=NoImage
     mirroring=NoMirror
     resourceName=None
     resourceObject=null
     tiling=NoTile
     imageSize=<150 150>
     margin=<ltrb 0 0 0 0>
     outlineSettings {
      bUseBrushTransparency=false
      roundingType=HalfHeightRadius
      width=0
      color {colorUseRule=UseColor_Specified, specifiedColor=<rgba 0 0 0 0>}
      cornerRadii=<0 0 0 1>
     }
     tintColor {colorUseRule=UseColor_Specified, specifiedColor=<rgba 1 1 1 1>}
     uVRegion {bIsValid=false, max=<0 0>, min=<0 0>}
    }
    triggeringInputAction {dataTable=&InputActionDataTable, rowName=GenericRightStickClick}
   }
   slot {class=HorizontalBoxSlot, properties {horizontalAlignment=HAlign_Left, verticalAlignment=VAlign_Bottom, padding=<ltrb 0 0 10 16>}}
  }
  Overlay_Remove {
   widgetClassPath=&Overlay
   bIsVariable=true
   parent=HBox_ListActions
   widget=Overlay_Remove
   slot {class=HorizontalBoxSlot, properties {padding=<ltrb 0 0 10 16>}}
  }
  Button_Remove {
   widgetClassPath=&IconTextButton
   bInherited=true
   bIsVariable=true
   parent=Overlay_Remove
   widget=Button_Remove
   properties {
    bIsFocusable=false
    buttonText=REMOVE
    contentAlignment=HAlign_Center
    minWidth=172
    onButtonBaseClicked=(&self:WidgetTree.Button_Remove.HandleOnButtonClicked)
    onButtonBaseDoubleClicked=(&self:WidgetTree.Button_Remove.HandleOnButtonDoubleClicked)
    onButtonBaseHovered=(&self:WidgetTree.Button_Remove.HandleOnButtonHovered)
    onButtonBaseUnhovered=(&self:WidgetTree.Button_Remove.HandleOnButtonUnhovered)
    onSelectedChangedBase=(&self:WidgetTree.Button_Remove.HandleOnSelectedChanged)
    style=&Athena_ButtonStyle_AngledBlueMenuButton_C
    visibility=Collapsed
    iconBrush {
     bIsDynamicallyLoaded=false
     drawAs=NoDrawType
     imageType=NoImage
     mirroring=NoMirror
     resourceName=None
     resourceObject=null
     tiling=NoTile
     imageSize=<150 150>
     margin=<ltrb 0 0 0 0>
     outlineSettings {
      bUseBrushTransparency=false
      roundingType=HalfHeightRadius
      width=0
      color {colorUseRule=UseColor_Specified, specifiedColor=<rgba 0 0 0 0>}
      cornerRadii=<0 0 0 1>
     }
     tintColor {colorUseRule=UseColor_Specified, specifiedColor=<rgba 1 1 1 1>}
     uVRegion {bIsValid=false, max=<0 0>, min=<0 0>}
    }
    triggeringInputAction {dataTable=&InputActionDataTable, rowName=GenericFaceButtonTop}
   }
   slot {class=OverlaySlot}
  }
  MatchmakingSpinner {
   widgetClassPath=&Image
   bIsVariable=true
   parent=HorizontalBox_0
   widget=MatchmakingSpinner
   properties {
    renderOpacity=0
    brush {
     bIsDynamicallyLoaded=false
     drawAs=Image
     imageType=NoImage
     mirroring=NoMirror
     resourceName=None
     resourceObject=&MI_UIKit_Monochrome_Throbber
     tiling=NoTile
     imageSize=<40 40>
     margin=<ltrb 0 0 0 0>
     outlineSettings {
      bUseBrushTransparency=false
      roundingType=HalfHeightRadius
      width=0
      color {colorUseRule=UseColor_Specified, specifiedColor=<rgba 0 0 0 0>}
      cornerRadii=<0 0 0 0>
     }
     tintColor {colorUseRule=UseColor_Specified, specifiedColor=<rgba 1 1 1 1>}
     uVRegion {bIsValid=false, max=<0 0>, min=<0 0>}
    }
    colorAndOpacity=<rgba 1 0.9301 0 1>
   }
   slot {class=HorizontalBoxSlot, properties {horizontalAlignment=HAlign_Center, verticalAlignment=VAlign_Bottom, padding=<ltrb 4 0 4 26>}}
  }
  Button_Select {
   widgetClassPath=&IconTextButton
   bInherited=true
   bIsVariable=true
   parent=HorizontalBox_0
   widget=Button_Select
   properties {
    bIsFocusable=false
    buttonText=LAUNCH
    contentAlignment=HAlign_Center
    minHeight=72
    minWidth=256
    onButtonBaseClicked=(&self:WidgetTree.Button_Select.HandleOnButtonClicked)
    onButtonBaseDoubleClicked=(&self:WidgetTree.Button_Select.HandleOnButtonDoubleClicked)
    onButtonBaseHovered=(&self:WidgetTree.Button_Select.HandleOnButtonHovered)
    onButtonBaseUnhovered=(&self:WidgetTree.Button_Select.HandleOnButtonUnhovered)
    onSelectedChangedBase=(&self:WidgetTree.Button_Select.HandleOnSelectedChanged)
    style=/FortUILibrary/Foundation/Buttons/ButtonStyles/ButtonStyle_Feature_M_Yellow.ButtonStyle_Feature_M_Yellow_C
    visibility=Visible
    triggeringInputAction {dataTable=&InputActionDataTable, rowName=Launch}
   }
   slot {
    class=HorizontalBoxSlot
    properties {horizontalAlignment=HAlign_Right, verticalAlignment=VAlign_Bottom, padding=<ltrb 0 0 10 16>, size {sizeRule=Fill, value=1}}
   }
  }
  Button_Cancel {
   widgetClassPath=&IconTextButton
   bInherited=true
   bIsVariable=true
   parent=HorizontalBox_0
   widget=Button_Cancel
   properties {
    bIsFocusable=false
    buttonText=CANCEL
    contentAlignment=HAlign_Center
    minWidth=172
    onButtonBaseClicked=(&self:WidgetTree.Button_Cancel.HandleOnButtonClicked)
    onButtonBaseDoubleClicked=(&self:WidgetTree.Button_Cancel.HandleOnButtonDoubleClicked)
    onButtonBaseHovered=(&self:WidgetTree.Button_Cancel.HandleOnButtonHovered)
    onButtonBaseUnhovered=(&self:WidgetTree.Button_Cancel.HandleOnButtonUnhovered)
    onSelectedChangedBase=(&self:WidgetTree.Button_Cancel.HandleOnSelectedChanged)
    style=&Athena_ButtonStyle_AngledBlueMenuButton_C
    triggeringInputAction {dataTable=/FortUILibrary/Input/CreativeInputActionDataTable, rowName=Cancel}
   }
   slot {class=HorizontalBoxSlot, properties {horizontalAlignment=HAlign_Right, verticalAlignment=VAlign_Bottom, padding=<ltrb 0 0 0 16>}}
  }
  VisibilityBorder_Touch {
   widgetClassPath=&CommonHardwareVisibilityBorder
   parent=Overlay_1
   widget=VisibilityBorder_Touch
   properties {visibilityQuery {
     autoDescription=" ANY( Input.Touch )"
     tokenStreamVersion=0
     userDescription=""
     queryTokenStream [0,1,1,1,0]
     tagDictionary [{tagName=Input.Touch}]
    }}
   slot {class=OverlaySlot, properties {horizontalAlignment=HAlign_Right, padding=<ltrb 0 16 16 0>}}
  }
  Button_CloseTouch {
   widgetClassPath=/FortUILibrary/Foundation/Buttons/CloseButton
   bIsVariable=true
   parent=VisibilityBorder_Touch
   widget=Button_CloseTouch
   properties {
    onButtonBaseClicked=(&self:WidgetTree.Button_CloseTouch.HandleOnButtonClicked)
    onButtonBaseDoubleClicked=(&self:WidgetTree.Button_CloseTouch.HandleOnButtonDoubleClicked)
    onButtonBaseHovered=(&self:WidgetTree.Button_CloseTouch.HandleOnButtonHovered)
    onButtonBaseUnhovered=(&self:WidgetTree.Button_CloseTouch.HandleOnButtonUnhovered)
    onSelectedChangedBase=(&self:WidgetTree.Button_CloseTouch.HandleOnSelectedChanged)
   }
   slot {class=BorderSlot, properties {padding=<ltrb 0 0 0 0>}}
  }
  Launch_Error {
   widgetClassPath=&LightningViolator
   bIsVariable=true
   parent=Overlay_3
   widget=Launch_Error
   properties {
    defaultViolatorText=Party too large
    showOnTop=true
    tailHorizontalAlignment=HAlign_Center
    visibility=Hidden
    renderTransform {angle=0, scale=<1 1>, shear=<0 0>, translation=<476 401>}
   }
   slot {class=OverlaySlot, properties {horizontalAlignment=HAlign_Center, verticalAlignment=VAlign_Center}}
  }
  Vbox_DELETEME {
   widgetClassPath=&VerticalBox
   parent=Overlay_3
   widget=Vbox_DELETEME
   properties {visibility=Collapsed}
   slot {class=OverlaySlot, properties {horizontalAlignment=HAlign_Fill, verticalAlignment=VAlign_Bottom, padding=<ltrb 0 10 0 10>}}
  }
  Text_Title {
   widgetClassPath=&CommonTextBlock
   parent=Vbox_DELETEME
   widget=Text_Title
   properties {
    text=Title of this Island
    font {
     bForceMonospaced=false
     bMaterialIsStencil=false
     fontMaterial=null
     fontObject=&BurbankSmall
     letterSpacing=0
     monospacedWidth=1
     size=24
     skewAmount=0
     typefaceFontName=Black
     outlineSettings {
      bApplyOutlineToDropShadows=false
      bMiteredCorners=false
      bSeparateFillAlpha=false
      outlineMaterial=null
      outlineSize=0
      outlineColor=<rgba 0 0 0 1>
     }
    }
    renderTransform {angle=0, scale=<1 1>, shear=<-6 0>, translation=<0 0>}
    shadowColorAndOpacity=<rgba 0 0 0 0.3>
    shadowOffset=<4.9266 5.2555>
   }
   slot {class=VerticalBoxSlot, properties {padding=<ltrb 0 20 0 0>}}
  }
  Text_Description {
   widgetClassPath=&CommonTextBlock
   parent=Vbox_DELETEME
   widget=Text_Description
   properties {
    autoWrapText=true
    text="More information about this island, this could be a longer description"
    font {
     bForceMonospaced=false
     bMaterialIsStencil=false
     fontMaterial=null
     fontObject=&BurbankSmall
     letterSpacing=0
     monospacedWidth=1
     size=24
     skewAmount=0
     typefaceFontName=Black
     outlineSettings {
      bApplyOutlineToDropShadows=false
      bMiteredCorners=false
      bSeparateFillAlpha=false
      outlineMaterial=null
      outlineSize=0
      outlineColor=<rgba 0 0 0 1>
     }
    }
    shadowColorAndOpacity=<rgba 0 0 0 0.3>
   }
   slot {class=VerticalBoxSlot}
  }
  Button_Refresh {
   widgetClassPath=&IconTextButton
   bInherited=true
   bIsVariable=true
   parent=Vbox_DELETEME
   widget=Button_Refresh
   properties {
    bIsFocusable=false
    buttonText=REFRESH
    contentAlignment=HAlign_Center
    minWidth=172
    onButtonBaseClicked=(&self:WidgetTree.Button_Refresh.HandleOnButtonClicked)
    onButtonBaseDoubleClicked=(&self:WidgetTree.Button_Refresh.HandleOnButtonDoubleClicked)
    onButtonBaseHovered=(&self:WidgetTree.Button_Refresh.HandleOnButtonHovered)
    onButtonBaseUnhovered=(&self:WidgetTree.Button_Refresh.HandleOnButtonUnhovered)
    onSelectedChangedBase=(&self:WidgetTree.Button_Refresh.HandleOnSelectedChanged)
    style=&Athena_ButtonStyle_AngledBlueMenuButton_C
    visibility=Collapsed
    iconBrush {
     bIsDynamicallyLoaded=false
     drawAs=NoDrawType
     imageType=NoImage
     mirroring=NoMirror
     resourceName=None
     resourceObject=null
     tiling=NoTile
     imageSize=<150 150>
     margin=<ltrb 0 0 0 0>
     outlineSettings {
      bUseBrushTransparency=false
      roundingType=HalfHeightRadius
      width=0
      color {colorUseRule=UseColor_Specified, specifiedColor=<rgba 0 0 0 0>}
      cornerRadii=<0 0 0 1>
     }
     tintColor {colorUseRule=UseColor_Specified, specifiedColor=<rgba 1 1 1 1>}
     uVRegion {bIsValid=false, max=<0 0>, min=<0 0>}
    }
    triggeringInputAction {dataTable=&InputActionDataTable, rowName=GenericLeftStickClick}
   }
   slot {class=VerticalBoxSlot, properties {horizontalAlignment=HAlign_Left, verticalAlignment=VAlign_Bottom, padding=<ltrb 0 0 10 16>}}
  }
  SizeBox_4 {
   widgetClassPath=&SizeBox
   parent=Vbox_DELETEME
   widget=SizeBox_4
   properties {heightOverride=400, widthOverride=640}
   slot {class=VerticalBoxSlot, properties {horizontalAlignment=HAlign_Right, verticalAlignment=VAlign_Top}}
  }
  Border_IslandImage {
   widgetClassPath=&Border
   bIsVariable=true
   parent=SizeBox_4
   widget=Border_IslandImage
   properties {
    background {
     bIsDynamicallyLoaded=false
     drawAs=NoDrawType
     imageType=NoImage
     mirroring=NoMirror
     resourceName=None
     resourceObject=null
     tiling=NoTile
     imageSize=<32 32>
     margin=<ltrb 0 0 0 0>
     outlineSettings {
      bUseBrushTransparency=false
      roundingType=HalfHeightRadius
      width=0
      color {colorUseRule=UseColor_Specified, specifiedColor=<rgba 0 0 0 0>}
      cornerRadii=<0 0 0 0>
     }
     tintColor {colorUseRule=UseColor_Specified, specifiedColor=<rgba 1 1 1 1>}
     uVRegion {bIsValid=false, max=<0 0>, min=<0 0>}
    }
    padding=<ltrb 0 0 0 0>
   }
   slot {class=SizeBoxSlot}
  }
  Button_CloseMobile {widgetClassPath=&CommonButtonLegacy, bInherited=true, widget=null}
 ]
}
