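# Translucent UI material for an action triangle in Creative UI, built from a texture threshold and procedural edge masks.
# The red channel of T_UI_ActionTriangle, scaled about center by Scale, feeds AAThreshold with Edge and EdgeSoftness parameters.
# A separate branch builds three half-plane masks from rotated UVs and SmoothStep, each with its own rotation and edge, then combines them with Max and inverts the result.
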Material /FortniteCoreUI/UI/Creative/Materials/M_UI_ActionTriangle
graph {
 properties {
  bUsedWithStaticMesh=false
  blendMode=BLEND_Translucent
  materialDomain=MD_UI
  pixelDepthOffsetMode=PDOM_Legacy
  usedShadingModels=Default Lit
 }
 nodes [
  Constant3Vector() {constant=<rgba 1 1 1 0>}
  n1 = ScalarParameter() {defaultValue=1, parameterName=Scale}
  MaterialFunctionCall!(
   Threshold=ScalarParameter() {defaultValue=0.2, parameterName=Edge}
   Gradient=TextureSample(UVs=MaterialFunctionCall(Texture Scale=@n1) {materialFunction=&ScaleUVsByCenter}.UVs) {texture=&/T_UI_ActionTriangle}.R
   Filter Width Override=ScalarParameter() {defaultValue=0.001, parameterName=EdgeSoftness}
   Filter Width Fudge Factor
  ) {materialFunction=/Game/UI/Materials/AAThreshold}
  n2 = MaterialFunctionCall(UVs=TextureCoordinate(), Rotation Center=Constant2Vector(), "Rotation Angle (0-1)"=ScalarParameter() {parameterName=Rotation1}) {materialFunction=&CustomRotator}
  n3 = ScalarParameter() {defaultValue=0.5, parameterName=Edge1}
  n4 = ScalarParameter() {defaultValue=0.001, parameterName=EdgeSoftness}
  n5 = MaterialFunctionCall(UVs=TextureCoordinate(), Rotation Center=Constant2Vector(), "Rotation Angle (0-1)"=ScalarParameter() {parameterName=Rotation2}) {materialFunction=&CustomRotator}
  n6 = ScalarParameter() {defaultValue=0.5, parameterName=Edge2}
  n7 = ScalarParameter() {defaultValue=0.001, parameterName=EdgeSoftness}
  n8 = Max(
   A=MaterialFunctionCall(Alpha=ComponentMask(None=@n2.Rotated Values) {g=true}, Min=@n3, Max=Add(A=@n3, B=@n4)) {materialFunction=&SmoothStep}.Result
   B=MaterialFunctionCall(Alpha=ComponentMask(None=@n5.Rotated Values) {g=true}, Min=@n6, Max=Add(A=@n6, B=@n7)) {materialFunction=&SmoothStep}.Result
  )
  n9 = TextureCoordinate()
  n10 = Constant2Vector()
  n11 = ScalarParameter() {parameterName=Rotation3}
  n12 = ScalarParameter() {defaultValue=0.5, parameterName=Edge3}
  n13 = MaterialFunctionCall(
   Alpha=ComponentMask(None=MaterialFunctionCall(UVs=@n9, Rotation Center=@n10, "Rotation Angle (0-1)"=@n11) {materialFunction=&CustomRotator}.Rotated Values) {g=true}
   Min=@n12
   Max=Add(A=@n12, B=ScalarParameter() {defaultValue=0.001, parameterName=EdgeSoftness})
  ) {materialFunction=&SmoothStep}
  OneMinus(None=Multiply(A=Max(A=@n8, B=@n13.Result)))
 ]
}
