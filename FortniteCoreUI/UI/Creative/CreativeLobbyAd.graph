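# Creative lobby ad tile widget, based on FortCreativeAdTile, showing a CMS-driven promo with an image, header, badge and description.
# OnCMSDataUpdated breaks CMSAdEntry, uppercases Header, Sub_Header and Description into their text blocks, and collapses the Violator badge when Sub_Header is empty.
# The image is an inherited EpicCMSImage not set in this graph; PreConstruct branches on IsDesignTime to set placeholder badge text.
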WidgetBlueprint /FortniteCoreUI/UI/Creative/CreativeLobbyAd
graphs [EventGraph {nodes [
   Event OnCMSDataUpdated() {is=title K2Node_Event}
   n1 = Break Creative Ad Data!(CreativeAdData:"Creative Ad Data Structure (by ref)"=Get CMSAdEntry() -> (CMSAdEntry:Creative Ad Data Structure)) -> (
    Header:String
    Sub_Header:String
    Description:String
    Creator_Name:String
    Island_Code:String
    Ad_Type:EFortCreativeAdType Enum
    Ad_Color_Preset:EFortCreativeAdColorPreset Enum
    Image:String
   ) {is=title K2Node_BreakStruct}
   > "SetText (Text)"(
    self=Get Text_Header() -> (Text_Header:Common Text Object Reference)
    InText=TextToUpper!(InText:"Text (by ref)"="To Text (String)"(InString=@n1)) -> (ReturnValue:Text)
   )
   n2 = TextToUpper(InText="To Text (String)"(InString=@n1.Sub_Header))
   > "SetText (Text)"(self=Get Text_SubHeader() -> (Text_SubHeader:Common Text Object Reference), InText=@n2)
   > SetVisibility(
    self=Get Violator() -> (Violator:Common Border Object Reference)
    InVisibility=Select(Option 0=SelfHitTestInvisible, Option 1=Collapsed, Index=TextIsEmpty(InText=@n2))
   )
   > "SetText (Text)"(
    self=Get Text_Description() -> (Text_Description:Common Text Object Reference)
    InText=TextToUpper(InText="To Text (String)"(InString=@n1.Description))
   )
   Branch(Condition=Event PreConstruct().IsDesignTime)
   > "SetText (Text)"(self=Get Text_SubHeader(), InText="LOCTEXT(\"New!\")")
   Event Construct()
   Event Tick()
  ]}]
variables []
widgets {
 info {inheritedWidgetCount=1, parentClass=/Script/FortniteUI.FortCreativeAdTile, rootWidgetClass=&SizeBox, widgetCount=17}
 widgets [
  SizeBox_4 {widgetClassPath=&SizeBox, widget=SizeBox_4, properties {bOverride_WidthOverride=true, widthOverride=576}}
  Overlay_0 {widgetClassPath=&Overlay, parent=SizeBox_4, widget=Overlay_0, slot {class=SizeBoxSlot}}
  HorizontalBox_1 {
   widgetClassPath=&HorizontalBox
   parent=Overlay_0
   widget=HorizontalBox_1
   slot {class=OverlaySlot, properties {horizontalAlignment=HAlign_Fill}}
  }
  ImageViolator {
   widgetClassPath=&SizeBox
   parent=HorizontalBox_1
   widget=ImageViolator
   properties {bOverride_HeightOverride=true, bOverride_WidthOverride=true, heightOverride=112, widthOverride=112}
   slot {class=HorizontalBoxSlot, properties {horizontalAlignment=HAlign_Left}}
  }
  Overlay_5 {widgetClassPath=&Overlay, parent=ImageViolator, widget=Overlay_5, slot {class=SizeBoxSlot}}
  SizeBox_2 {
   widgetClassPath=&SizeBox
   parent=Overlay_5
   widget=SizeBox_2
   properties {bOverride_HeightOverride=true, bOverride_WidthOverride=true, heightOverride=112, widthOverride=112}
   slot {class=OverlaySlot, properties {horizontalAlignment=HAlign_Center, verticalAlignment=VAlign_Center}}
  }
  Image_CreativeAd {
   widgetClassPath=/Script/EpicCMSUIFramework.EpicCMSImage
   bInherited=true
   bIsVariable=true
   parent=SizeBox_2
   widget=Image_CreativeAd
   properties {brush {
     bIsDynamicallyLoaded=false
     drawAs=Image
     imageType=NoImage
     mirroring=NoMirror
     resourceName=None
     resourceObject=/Game/UI/Foundation/Textures/Icons/Cards/SquadTypes/T-Icon-Chest-128
     tiling=NoTile
     imageSize=<128 128>
     margin=<ltrb 0 0 0 0>
     outlineSettings {
      bUseBrushTransparency=false
      roundingType=HalfHeightRadius
      width=0
      color {colorUseRule=UseColor_Specified, specifiedColor=<rgba 0 0 0 0>}
      cornerRadii=<0 0 0 0>
     }
     tintColor {colorUseRule=UseColor_Specified, specifiedColor=<rgba 1 1 1 1>}
     uVRegion {bIsValid=false, max=<0 0>, min=<0 0>}
    }}
   slot {class=SizeBoxSlot}
  }
  TextBox {
   widgetClassPath=&SizeBox
   parent=HorizontalBox_1
   widget=TextBox
   properties {widthOverride=392}
   slot {class=HorizontalBoxSlot, properties {verticalAlignment=VAlign_Center, padding=<ltrb 8 0 0 4>, size {sizeRule=Fill, value=1}}}
  }
  Text {widgetClassPath=&VerticalBox, parent=TextBox, widget=Text, slot {class=SizeBoxSlot, properties {verticalAlignment=VAlign_Top}}}
  HorizontalBox_0 {
   widgetClassPath=&HorizontalBox
   parent=Text
   widget=HorizontalBox_0
   slot {class=VerticalBoxSlot, properties {verticalAlignment=VAlign_Center}}
  }
  Violator {
   widgetClassPath=&CommonBorder
   bIsVariable=true
   parent=HorizontalBox_0
   widget=Violator
   properties {
    background {
     bIsDynamicallyLoaded=false
     drawAs=Box
     imageType=NoImage
     mirroring=NoMirror
     resourceName=None
     resourceObject=/Game/UI/Login/SubGameSelect/Materials/M_Notification_Right
     tiling=NoTile
     imageSize=<96 32>
     margin=<ltrb 0.1 0 0.25 0>
     outlineSettings {
      bUseBrushTransparency=false
      roundingType=HalfHeightRadius
      width=0
      color {colorUseRule=UseColor_Specified, specifiedColor=<rgba 0 0 0 0>}
      cornerRadii=<0 0 0 0>
     }
     tintColor {colorUseRule=UseColor_Specified, specifiedColor=<rgba 1 1 1 1>}
     uVRegion {bIsValid=false, max=<0 0>, min=<0 0>}
    }
    padding=<ltrb 16 4 24 4>
   }
   slot {class=HorizontalBoxSlot, properties {horizontalAlignment=HAlign_Left, verticalAlignment=VAlign_Center}}
  }
  SizeBox_1 {
   widgetClassPath=&SizeBox
   parent=Violator
   widget=SizeBox_1
   properties {bOverride_MaxDesiredWidth=true, maxDesiredWidth=256}
   slot {class=BorderSlot, properties {padding=<ltrb 16 4 24 4>}}
  }
  Text_SubHeader {
   widgetClassPath=&CommonTextBlock
   bIsVariable=true
   parent=SizeBox_1
   widget=Text_SubHeader
   properties {
    scrollStyle=&TextScrollStyle-Base_C
    text="NEW!"
    textTransformPolicy=ToUpper
    font {
     bForceMonospaced=false
     bMaterialIsStencil=false
     fontMaterial=null
     fontObject=&Burbank
     letterSpacing=0
     monospacedWidth=1
     size=16
     skewAmount=0
     typefaceFontName=Burbank-Black
     outlineSettings {
      bApplyOutlineToDropShadows=false
      bMiteredCorners=false
      bSeparateFillAlpha=false
      outlineMaterial=null
      outlineSize=0
      outlineColor=<rgba 0 0 0 1>
     }
    }
    shadowOffset=<0 0>
   }
   slot {class=SizeBoxSlot}
  }
  SizeBox_3 {
   widgetClassPath=&SizeBox
   parent=HorizontalBox_0
   widget=SizeBox_3
   properties {widthOverride=376}
   slot {class=HorizontalBoxSlot, properties {padding=<ltrb 8 0 0 0>, size {sizeRule=Fill, value=1}}}
  }
  ScaleBox_0 {
   widgetClassPath=&ScaleBox
   parent=SizeBox_3
   widget=ScaleBox_0
   properties {stretchDirection=DownOnly}
   slot {class=SizeBoxSlot, properties {verticalAlignment=VAlign_Center}}
  }
  Text_Header {
   widgetClassPath=&CommonTextBlock
   bIsVariable=true
   parent=ScaleBox_0
   widget=Text_Header
   properties {
    text="{Title}"
    font {
     bForceMonospaced=false
     bMaterialIsStencil=false
     fontMaterial=null
     fontObject=&BurbankSmall
     letterSpacing=0
     monospacedWidth=1
     size=24
     skewAmount=0
     typefaceFontName=Black
     outlineSettings {
      bApplyOutlineToDropShadows=false
      bMiteredCorners=false
      bSeparateFillAlpha=false
      outlineMaterial=null
      outlineSize=0
      outlineColor=<rgba 0 0.0417 0.5 0>
     }
    }
    renderTransform {angle=0, scale=<1 1>, shear=<-8 0>, translation=<0 -2>}
    shadowOffset=<0 0>
   }
   slot {class=ScaleBoxSlot, properties {horizontalAlignment=HAlign_Fill}}
  }
  Text_Description {
   widgetClassPath=&CommonTextBlock
   bIsVariable=true
   parent=Text
   widget=Text_Description
   properties {
    text="{Description}"
    wrapTextAt=452
    colorAndOpacity {colorUseRule=UseColor_Specified, specifiedColor=<rgba 0 1 1 1>}
    font {
     bForceMonospaced=false
     bMaterialIsStencil=false
     fontMaterial=null
     fontObject=&BurbankSmall
     letterSpacing=0
     monospacedWidth=1
     size=15
     skewAmount=0
     typefaceFontName=Black
     outlineSettings {
      bApplyOutlineToDropShadows=false
      bMiteredCorners=false
      bSeparateFillAlpha=false
      outlineMaterial=null
      outlineSize=0
      outlineColor=<rgba 0 0 0 1>
     }
    }
    renderTransform {angle=0, scale=<1 1>, shear=<-8 0>, translation=<0 0>}
    shadowOffset=<0 0>
   }
   slot {class=VerticalBoxSlot, properties {verticalAlignment=VAlign_Top, padding=<ltrb 4 0 0 0>, size {sizeRule=Fill, value=1}}}
  }
 ]
}
